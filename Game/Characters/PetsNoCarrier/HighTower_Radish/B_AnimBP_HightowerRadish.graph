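# Animation Blueprint for the hovering HighTower Radish pet, driven by the owning player's movement, targeting, crouch and skydive state.
# A Default state machine (nested locomotion with underwater and random-break variants, skydive, DBNO, emote and melee reactions) or a FrontEnd preview machine feeds DefaultSlot.
# Additive poses layer crouch, lean and direction changes.
# The cached result then gets pelvis/root offsets, left-hand IK, and root shrinking when DBNO-carried.
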AnimBlueprint /Game/Characters/PetsNoCarrier/HighTower_Radish/B_AnimBP_HightowerRadish
target_skeleton=/Game/Characters/PetsNoCarrier/HighTower_Radish/Meshes/M_MED_HighTower_Radish_Pet_Skeleton
graphs [
 AnimGraph {nodes [
   n1 = Get BodyOffset_Total() -> (BodyOffset_Total:Vector)
   n2 = "Transform (Modify) Bone - Bone: pelvis"!(ComponentPose=Local To Component(LocalPose=Use cached pose 'SavedPose19943'()), Alpha=1, Translation=@n1, Rotation="0,0,0") -> (Pose) {rotationMode=BMM_Additive, translationMode=BMM_Additive, alphaBoolBlend {blendOption=Linear}, boneToModify {boneName=pelvis}}
   n3 = "Transform (Modify) Bone - Bone: ik_hand_root"!(ComponentPose=@n2, Alpha=1, Translation=@n1, Rotation="0,0,0") -> (Pose) {rotationMode=BMM_Additive, translationMode=BMM_Additive, alphaBoolBlend {blendOption=Linear}, boneToModify {boneName=ik_hand_root}}
   n4 = "Transform (Modify) Bone - Bone: root"!(
    ComponentPose=Local To Component(LocalPose=Component To Local(ComponentPose=@n3))
    Alpha
    Rotation_Roll:float=0
    Rotation_Pitch:float=0
    Rotation_Yaw:float=Get OffsetRootYawSmooth() -> (OffsetRootYawSmooth:float)
   ) -> (Pose) {rotationMode=BMM_Additive, alphaBoolBlend {blendOption=Linear}, boneToModify {boneName=root}}
   n5 = Get LeftHandIKCurve() -> (LeftHandIKCurve:float)
   n6 = Leg IK!(
    ComponentPose:Component Space Pose Link Structure=Copy Bone - Source Bone: hand_r - Target Bone: ik_hand_gun!(
     ComponentPose:Component Space Pose Link Structure=Copy Bone - Source Bone: VB hand_r_weapon - Target Bone: weapon!(
      ComponentPose:Component Space Pose Link Structure=@n4
      Alpha:float=1
      bCopyTranslation:bool=true
      bCopyRotation:bool=true
      bCopyScale:bool=False
     ) -> (Pose:Component Space Pose Link Structure) {is=title AnimGraphNode_CopyBone, alphaBoolBlend {blendOption=Linear}, sourceBone {boneName=VB hand_r_weapon}, targetBone {boneName=weapon}}
     Alpha:float=@n5
     bCopyTranslation:bool=true
     bCopyRotation:bool=true
     bCopyScale:bool=False
    ) -> (Pose:Component Space Pose Link Structure) {is=title AnimGraphNode_CopyBone, alphaBoolBlend {blendOption=Linear}, sourceBone {boneName=hand_r}, targetBone {boneName=ik_hand_gun}}
    Alpha:float=@n5
   ) -> (Pose:Component Space Pose Link Structure) {
    is=title AnimGraphNode_LegIK
    alphaBoolBlend {blendOption=Linear}
    legsDefinition [{
      bEnableKneeTwistCorrection=false
      bEnableRotationLimit=false
      footBoneForwardAxis=X
      hingeRotationAxis=Z
      minRotationAngle=15
      numBonesInLimb=2
      twistOffsetCurveName=None
      fKFootBone {boneName=hand_l}
      iKFootBone {boneName=ik_hand_l}
     }]
   }
   n7 = "Transform (Modify) Bone - Bone: root"!(ComponentPose=@n6, Alpha=Get InVehicle() -> (InVehicle:float), Scale:Vector="0,0,0") -> (Pose) {scaleMode=BMM_Additive, alphaBoolBlend {blendOption=Linear}, boneToModify {boneName=root}, scale=<0 0 0>}
   n8 = "Transform (Modify) Bone - Bone: root"!(
    ComponentPose=@n7
    Alpha=Get IsDBNOCarriedAlpha() -> (IsDBNOCarriedAlpha:double)
    Translation:Vector="0,0,0"
    Rotation:Rotator="0,0,0"
    Scale="0.01,0.01,0.01"
   ) -> (Pose) {
    scaleMode=BMM_Replace
    scaleSpace=BCS_WorldSpace
    translationMode=BMM_Replace
    alphaBoolBlend {blendOption=Linear}
    boneToModify {boneName=root}
    scale=<0.01 0.01 0.01>
   }
   Output Pose(Result=Component To Local(ComponentPose=@n8)) {layerGroup=None}
   AnimGraphNode_StateMachine_0 = Default() {is=title AnimGraphNode_StateMachine}
   AnimGraphNode_StateMachine_1 = FrontEnd() {is=title AnimGraphNode_StateMachine}
   n9 = Get bIsFrontEndPreview() -> (bIsFrontEndPreview:bool)
   Save cached pose 'SavedPose19943'(Pose=Slot 'DefaultSlot'(Source=Blend Poses by bool(BlendPose_0=@AnimGraphNode_StateMachine_1, BlendPose_1=@AnimGraphNode_StateMachine_0, bActiveValue=@n9) {blendType=Linear})) {cacheName=SavedPose19943}
  ]}
 Default {
  entry=Locomotion
  path=AnimGraph.AnimGraphNode_StateMachine_0.Default
  ids=id
  nodes [
   AnimStateNode_0 = Locomotion() {is=title AnimStateNode}
   AnimStateNode_1 = EmoteReactIntro() {is=title AnimStateNode}
   AnimStateNode_2 = EmoteReactLoop() {is=title AnimStateNode}
   AnimStateNode_3 = Skydive() {is=title AnimStateNode}
   AnimStateNode_4 = DBNO() {is=title AnimStateNode}
   AnimStateNode_5 = MeleeReact() {is=title AnimStateNode}
   AnimStateNode_6 = EmoteReactOutro() {is=title AnimStateNode}
   AnimStateTransitionNode_0 = AnimStateTransitionNode() {crossfadeDuration=0.5}
   AnimStateTransitionNode_1 = AnimStateTransitionNode() {crossfadeDuration=0.5}
   AnimStateTransitionNode_10 = AnimStateTransitionNode() {crossfadeDuration=0.4}
   AnimStateTransitionNode_11 = AnimStateTransitionNode() {bAutomaticRuleBasedOnSequencePlayerInState=true, crossfadeDuration=0.5}
   AnimStateTransitionNode_2 = AnimStateTransitionNode() {bAutomaticRuleBasedOnSequencePlayerInState=true}
   AnimStateTransitionNode_4 = AnimStateTransitionNode() {crossfadeDuration=0.5}
   AnimStateTransitionNode_5 = AnimStateTransitionNode() {crossfadeDuration=0.5}
   AnimStateTransitionNode_6 = AnimStateTransitionNode()
   AnimStateTransitionNode_7 = AnimStateTransitionNode()
   AnimStateTransitionNode_9 = AnimStateTransitionNode() {crossfadeDuration=0.33}
  ]
  links [
   {
    dst=Locomotion
    rule=AnimGraph.AnimGraphNode_StateMachine_0.Default.AnimStateTransitionNode_7.Transition
    src=DBNO
    via=AnimStateTransitionNode_7
   }
   {
    dst=EmoteReactLoop
    rule=AnimGraph.AnimGraphNode_StateMachine_0.Default.AnimStateTransitionNode_2.Transition
    src=EmoteReactIntro
    via=AnimStateTransitionNode_2
   }
   {
    dst=Locomotion
    rule=AnimGraph.AnimGraphNode_StateMachine_0.Default.AnimStateTransitionNode_1.Transition
    src=EmoteReactIntro
    via=AnimStateTransitionNode_1
   }
   {
    dst=EmoteReactOutro
    rule=AnimGraph.AnimGraphNode_StateMachine_0.Default.AnimStateTransitionNode_10.Transition
    src=EmoteReactLoop
    via=AnimStateTransitionNode_10
   }
   {
    dst=Locomotion
    rule=AnimGraph.AnimGraphNode_StateMachine_0.Default.AnimStateTransitionNode_11.Transition
    src=EmoteReactOutro
    via=AnimStateTransitionNode_11
   }
   {
    dst=DBNO
    rule=AnimGraph.AnimGraphNode_StateMachine_0.Default.AnimStateTransitionNode_6.Transition
    src=Locomotion
    via=AnimStateTransitionNode_6
   }
   {
    dst=EmoteReactIntro
    rule=AnimGraph.AnimGraphNode_StateMachine_0.Default.AnimStateTransitionNode_0.Transition
    src=Locomotion
    via=AnimStateTransitionNode_0
   }
   {
    dst=Skydive
    rule=AnimGraph.AnimGraphNode_StateMachine_0.Default.AnimStateTransitionNode_4.Transition
    src=Locomotion
    via=AnimStateTransitionNode_4
   }
   {
    dst=Locomotion
    rule=AnimGraph.AnimGraphNode_StateMachine_0.Default.AnimStateTransitionNode_9.Transition
    src=MeleeReact
    via=AnimStateTransitionNode_9
   }
   {
    dst=Locomotion
    rule=AnimGraph.AnimGraphNode_StateMachine_0.Default.AnimStateTransitionNode_5.Transition
    src=Skydive
    via=AnimStateTransitionNode_5
   }
  ]
 }
 Locomotion {
  path=AnimGraph.AnimGraphNode_StateMachine_0.Default.AnimStateNode_0.Locomotion
  nodes [
   AnimGraphNode_StateMachine_0 = Locomotion() {is=title AnimGraphNode_StateMachine}
   Output Animation Pose(Result=@AnimGraphNode_StateMachine_0)
  ]
 }
 Locomotion {
  entry=Idle
  path=AnimGraph.AnimGraphNode_StateMachine_0.Default.AnimStateNode_0.Locomotion.AnimGraphNode_StateMachine_0.Locomotion
  ids=id
  nodes [
   AnimStateNode_0 = Idle()
   AnimStateNode_1 = MoveStart() {
    is=title AnimStateNode
    stateEntered {guid=#1}
    stateFullyBlended {guid=#2}
    stateLeft {guid=#3}
   }
   AnimStateNode_2 = MoveLoop() {
    is=title AnimStateNode
    stateEntered {guid=#4}
    stateFullyBlended {guid=#5}
    stateLeft {guid=#6}
   }
   AnimStateNode_3 = MoveStop() {is=title AnimStateNode}
   AnimStateTransitionNode_0 = AnimStateTransitionNode() {bAutomaticRuleBasedOnSequencePlayerInState=true}
   AnimStateTransitionNode_1 = AnimStateTransitionNode()
   AnimStateTransitionNode_10 = AnimStateTransitionNode()
   AnimStateTransitionNode_2 = AnimStateTransitionNode()
   AnimStateTransitionNode_4 = AnimStateTransitionNode() {crossfadeDuration=0.4}
   AnimStateTransitionNode_5 = AnimStateTransitionNode() {bAutomaticRuleBasedOnSequencePlayerInState=true, crossfadeDuration=0.5}
   AnimStateTransitionNode_8 = AnimStateTransitionNode()
   AnimStateTransitionNode_9 = AnimStateTransitionNode() {crossfadeDuration=0.35}
  ]
  links [
   {
    dst=MoveStart
    rule=AnimGraph.AnimGraphNode_StateMachine_0.Default.AnimStateNode_0.Locomotion.AnimGraphNode_StateMachine_0.Locomotion.AnimStateTransitionNode_1.Transition
    src=Idle
    via=AnimStateTransitionNode_1
   }
   {
    dst=MoveStart
    rule=AnimGraph.AnimGraphNode_StateMachine_0.Default.AnimStateNode_0.Locomotion.AnimGraphNode_StateMachine_0.Locomotion.AnimStateTransitionNode_10.Transition
    src=Idle
    via=AnimStateTransitionNode_10
   }
   {
    dst=MoveStart
    rule=AnimGraph.AnimGraphNode_StateMachine_0.Default.AnimStateNode_0.Locomotion.AnimGraphNode_StateMachine_0.Locomotion.AnimStateTransitionNode_8.Transition
    src=Idle
    via=AnimStateTransitionNode_8
   }
   {
    dst=MoveStop
    rule=AnimGraph.AnimGraphNode_StateMachine_0.Default.AnimStateNode_0.Locomotion.AnimGraphNode_StateMachine_0.Locomotion.AnimStateTransitionNode_4.Transition
    src=MoveLoop
    via=AnimStateTransitionNode_4
   }
   {
    dst=MoveLoop
    rule=AnimGraph.AnimGraphNode_StateMachine_0.Default.AnimStateNode_0.Locomotion.AnimGraphNode_StateMachine_0.Locomotion.AnimStateTransitionNode_0.Transition
    src=MoveStart
    via=AnimStateTransitionNode_0
   }
   {
    dst=MoveStop
    rule=AnimGraph.AnimGraphNode_StateMachine_0.Default.AnimStateNode_0.Locomotion.AnimGraphNode_StateMachine_0.Locomotion.AnimStateTransitionNode_2.Transition
    src=MoveStart
    via=AnimStateTransitionNode_2
   }
   {
    dst=Idle
    rule=AnimGraph.AnimGraphNode_StateMachine_0.Default.AnimStateNode_0.Locomotion.AnimGraphNode_StateMachine_0.Locomotion.AnimStateTransitionNode_5.Transition
    src=MoveStop
    via=AnimStateTransitionNode_5
   }
   {
    dst=Idle
    rule=AnimGraph.AnimGraphNode_StateMachine_0.Default.AnimStateNode_0.Locomotion.AnimGraphNode_StateMachine_0.Locomotion.AnimStateTransitionNode_9.Transition
    src=MoveStop
    via=AnimStateTransitionNode_9
   }
  ]
 }
 Idle {
  path=AnimGraph.AnimGraphNode_StateMachine_0.Default.AnimStateNode_0.Locomotion.AnimGraphNode_StateMachine_0.Locomotion.AnimStateNode_0.Idle
  nodes [
   AnimGraphNode_StateMachine_0 = IdleSM() {is=title AnimGraphNode_StateMachine}
   Output Animation Pose(Result=@AnimGraphNode_StateMachine_0)
  ]
 }
 IdleSM {
  entry=Idle
  path=AnimGraph.AnimGraphNode_StateMachine_0.Default.AnimStateNode_0.Locomotion.AnimGraphNode_StateMachine_0.Locomotion.AnimStateNode_0.Idle.AnimGraphNode_StateMachine_0.IdleSM
  ids=id
  nodes [
   AnimStateNode_0 = Idle()
   AnimStateNode_1 = IdleUnderWater() {is=title AnimStateNode}
   AnimStateNode_2 = IdleUnderWaterExit() {is=title AnimStateNode}
   AnimStateTransitionNode_0 = AnimStateTransitionNode()
   AnimStateTransitionNode_1 = AnimStateTransitionNode()
   AnimStateTransitionNode_2 = AnimStateTransitionNode()
   AnimStateTransitionNode_3 = AnimStateTransitionNode() {bAutomaticRuleBasedOnSequencePlayerInState=true, crossfadeDuration=0.4}
   AnimStateTransitionNode_4 = AnimStateTransitionNode()
   AnimStateTransitionNode_5 = AnimStateTransitionNode()
  ]
  links [
   {
    dst=IdleUnderWater
    rule=AnimGraph.AnimGraphNode_StateMachine_0.Default.AnimStateNode_0.Locomotion.AnimGraphNode_StateMachine_0.Locomotion.AnimStateNode_0.Idle.AnimGraphNode_StateMachine_0.IdleSM.AnimStateTransitionNode_0.Transition
    src=Idle
    via=AnimStateTransitionNode_0
   }
   {
    dst=Idle
    rule=AnimGraph.AnimGraphNode_StateMachine_0.Default.AnimStateNode_0.Locomotion.AnimGraphNode_StateMachine_0.Locomotion.AnimStateNode_0.Idle.AnimGraphNode_StateMachine_0.IdleSM.AnimStateTransitionNode_4.Transition
    src=IdleUnderWater
    via=AnimStateTransitionNode_4
   }
   {
    dst=IdleUnderWaterExit
    rule=AnimGraph.AnimGraphNode_StateMachine_0.Default.AnimStateNode_0.Locomotion.AnimGraphNode_StateMachine_0.Locomotion.AnimStateNode_0.Idle.AnimGraphNode_StateMachine_0.IdleSM.AnimStateTransitionNode_1.Transition
    src=IdleUnderWater
    via=AnimStateTransitionNode_1
   }
   {
    dst=Idle
    rule=AnimGraph.AnimGraphNode_StateMachine_0.Default.AnimStateNode_0.Locomotion.AnimGraphNode_StateMachine_0.Locomotion.AnimStateNode_0.Idle.AnimGraphNode_StateMachine_0.IdleSM.AnimStateTransitionNode_3.Transition
    src=IdleUnderWaterExit
    via=AnimStateTransitionNode_3
   }
   {
    dst=Idle
    rule=AnimGraph.AnimGraphNode_StateMachine_0.Default.AnimStateNode_0.Locomotion.AnimGraphNode_StateMachine_0.Locomotion.AnimStateNode_0.Idle.AnimGraphNode_StateMachine_0.IdleSM.AnimStateTransitionNode_5.Transition
    src=IdleUnderWaterExit
    via=AnimStateTransitionNode_5
   }
   {
    dst=IdleUnderWater
    rule=AnimGraph.AnimGraphNode_StateMachine_0.Default.AnimStateNode_0.Locomotion.AnimGraphNode_StateMachine_0.Locomotion.AnimStateNode_0.Idle.AnimGraphNode_StateMachine_0.IdleSM.AnimStateTransitionNode_2.Transition
    src=IdleUnderWaterExit
    via=AnimStateTransitionNode_2
   }
  ]
 }
 Idle {
  path=AnimGraph.AnimGraphNode_StateMachine_0.Default.AnimStateNode_0.Locomotion.AnimGraphNode_StateMachine_0.Locomotion.AnimStateNode_0.Idle.AnimGraphNode_StateMachine_0.IdleSM.AnimStateNode_0.Idle
  nodes [
   Apply Additive() {alphaBoolBlend {blendOption=Linear}}
   Blend Poses by bool(
    BlendPose_0=Sequence Player 'Hover_Idle_Downsights'() {groupName=IdleSync, method=SyncGroup, sequence=/Game/Animation/Game/Pets/HightowerRadish/Backpack/Idle/Hover_Idle_Downsights}
    BlendTime_0=0.4
    BlendTime_1=0.4
    bActiveValue=Get bIsTargeting!(self:Fort Player Anim Instance Object Reference=Get OwnerAnimBP() -> (OwnerAnimBP:Fort Player Anim Instance Object Reference)) -> (bIsTargeting)
   ) {blendType=Linear}
   n1 = Sequence Player 'Hover_Idle_NonTargeted'() {
    groupName=IdleSync
    method=SyncGroup
    sequence=/Game/Animation/Game/Pets/HightowerRadish/Backpack/Idle/Hover_Idle_NonTargeted
    startPosition=1
   }
   n2 = "Sequence Player (additive) 'Hover_Idle_Crouch_Pose'"() {sequence=/Game/Animation/Game/Pets/HightowerRadish/Backpack/Idle/Hover_Idle_Crouch_Pose}
   Output Animation Pose(Result=Blend Poses by bool(
     BlendPose_0=Apply Mesh Space Additive!(Base:Pose Link Structure=@n1, Additive:Pose Link Structure=@n2, Alpha:float=1) -> (Pose:Pose Link Structure) {is=title AnimGraphNode_ApplyMeshSpaceAdditive, alphaBoolBlend {blendOption=Linear}}
     BlendPose_1=Sequence Player 'Hover_Idle_NonTargeted'() {groupName=IdleSync, method=SyncGroup, sequence=&Hover_Idle_NonTargeted}
     BlendTime_0=0.25
     BlendTime_1=0.3
     bActiveValue=Get bIsCrouching()
    ) {blendType=Linear})
  ]
 }
 IdleUnderWater {
  path=AnimGraph.AnimGraphNode_StateMachine_0.Default.AnimStateNode_0.Locomotion.AnimGraphNode_StateMachine_0.Locomotion.AnimStateNode_0.Idle.AnimGraphNode_StateMachine_0.IdleSM.AnimStateNode_1.IdleUnderWater
  nodes [
   n1 = "Sequence Player (additive) 'Hover_Idle_Underwater_Enter_Tail_Add'"!(PlayRate:float=2, bLoopAnimation:bool=False) -> (Pose) {bLoopAnimation=false, playRate=2, sequence=/Game/Animation/Game/Pets/HightowerRadish/Backpack/Idle/Hover_Idle_Underwater_Enter_Tail_Add}
   Output Animation Pose(Result=Apply Additive(
     Base=Sequence Player 'Hover_Idle_Underwater'() {groupName=IdleSync, method=SyncGroup, sequence=/Game/Animation/Game/Pets/HightowerRadish/Backpack/Idle/Hover_Idle_Underwater}
     Additive=@n1
    ) {alphaBoolBlend {blendOption=Linear}})
  ]
 }
 IdleUnderWaterExit {
  path=AnimGraph.AnimGraphNode_StateMachine_0.Default.AnimStateNode_0.Locomotion.AnimGraphNode_StateMachine_0.Locomotion.AnimStateNode_0.Idle.AnimGraphNode_StateMachine_0.IdleSM.AnimStateNode_2.IdleUnderWaterExit
  nodes [Output Animation Pose(Result=Sequence Player 'Hover_Idle_UnderWater_Exit'!(bLoopAnimation:bool=False) -> (Pose) {bLoopAnimation=false, sequence=/Game/Animation/Game/Pets/HightowerRadish/Backpack/Idle/Hover_Idle_UnderWater_Exit})]
 }
 Transition {
  path=AnimGraph.AnimGraphNode_StateMachine_0.Default.AnimStateNode_0.Locomotion.AnimGraphNode_StateMachine_0.Locomotion.AnimStateNode_0.Idle.AnimGraphNode_StateMachine_0.IdleSM.AnimStateTransitionNode_0.Transition
  nodes [Result(bCanEnterTransition=Get bPetUnderWaterAndNotTargeting() -> (bPetUnderWaterAndNotTargeting:bool))]
 }
 Transition {
  path=AnimGraph.AnimGraphNode_StateMachine_0.Default.AnimStateNode_0.Locomotion.AnimGraphNode_StateMachine_0.Locomotion.AnimStateNode_0.Idle.AnimGraphNode_StateMachine_0.IdleSM.AnimStateTransitionNode_1.Transition
  nodes [Result(bCanEnterTransition=NOT Boolean(A=Get bPetUnderwater() -> (bPetUnderwater:bool)))]
 }
 Transition {
  path=AnimGraph.AnimGraphNode_StateMachine_0.Default.AnimStateNode_0.Locomotion.AnimGraphNode_StateMachine_0.Locomotion.AnimStateNode_0.Idle.AnimGraphNode_StateMachine_0.IdleSM.AnimStateTransitionNode_2.Transition
  nodes [Result(bCanEnterTransition=Get bPetUnderWaterAndNotTargeting())]
 }
 Transition {
  path=AnimGraph.AnimGraphNode_StateMachine_0.Default.AnimStateNode_0.Locomotion.AnimGraphNode_StateMachine_0.Locomotion.AnimStateNode_0.Idle.AnimGraphNode_StateMachine_0.IdleSM.AnimStateTransitionNode_3.Transition
  nodes [Result()]
 }
 Transition {
  path=AnimGraph.AnimGraphNode_StateMachine_0.Default.AnimStateNode_0.Locomotion.AnimGraphNode_StateMachine_0.Locomotion.AnimStateNode_0.Idle.AnimGraphNode_StateMachine_0.IdleSM.AnimStateTransitionNode_4.Transition
  nodes [Result(bCanEnterTransition=Get bIsTargeting!() -> (bIsTargeting))]
 }
 Transition {
  path=AnimGraph.AnimGraphNode_StateMachine_0.Default.AnimStateNode_0.Locomotion.AnimGraphNode_StateMachine_0.Locomotion.AnimStateNode_0.Idle.AnimGraphNode_StateMachine_0.IdleSM.AnimStateTransitionNode_5.Transition
  nodes [Result(bCanEnterTransition=Get bIsTargeting())]
 }
 MoveStart {
  path=AnimGraph.AnimGraphNode_StateMachine_0.Default.AnimStateNode_0.Locomotion.AnimGraphNode_StateMachine_0.Locomotion.AnimStateNode_1.MoveStart
  nodes [
   Apply Additive() {alphaBoolBlend {blendOption=Linear}}
   n1 = Sequence Player 'Locomotion_FlightLoop_Intro'() {
    bLoopAnimation=false
    groupName=FlightLoop_Into
    method=SyncGroup
    sequence=/Game/Animation/Game/Pets/HightowerRadish/Backpack/Locomotion/Locomotion_FlightLoop_Intro
   }
   n2 = "Sequence Player (additive) 'Locomotion_FlightLoop_Crouch_Pose'"() {sequence=/Game/Animation/Game/Pets/HightowerRadish/Backpack/Locomotion/Locomotion_FlightLoop_Crouch_Pose}
   n3 = Sequence Player 'Locomotion_FlightLoop_Intro'() {
    bLoopAnimation=false
    groupName=FlightLoop_Into
    method=SyncGroup
    sequence=&Locomotion_FlightLoop_Intro
   }
   Output Animation Pose(Result=Blend Poses by bool(
     BlendPose_0=Apply Mesh Space Additive(Base=@n1, Additive=@n2, Alpha=1) {alphaBoolBlend {blendOption=Linear}}
     BlendPose_1=@n3
     BlendTime_0=0.4
     BlendTime_1=0.4
     bActiveValue=Get bIsCrouching()
    ) {blendType=Linear})
  ]
 }
 MoveLoop {
  path=AnimGraph.AnimGraphNode_StateMachine_0.Default.AnimStateNode_0.Locomotion.AnimGraphNode_StateMachine_0.Locomotion.AnimStateNode_2.MoveLoop
  nodes [
   AnimGraphNode_StateMachine_0 = MoveLoopSM() {is=title AnimGraphNode_StateMachine}
   Output Animation Pose(Result=@AnimGraphNode_StateMachine_0)
  ]
 }
 MoveLoopSM {
  entry=MoveLoop
  path=AnimGraph.AnimGraphNode_StateMachine_0.Default.AnimStateNode_0.Locomotion.AnimGraphNode_StateMachine_0.Locomotion.AnimStateNode_2.MoveLoop.AnimGraphNode_StateMachine_0.MoveLoopSM
  ids=id
  nodes [
   AnimStateNode_0 = MoveLoop()
   AnimStateNode_1 = MoveLoopUnderWater() {is=title AnimStateNode}
   AnimStateNode_2 = MoveLoopUnderWaterExit() {is=title AnimStateNode}
   AnimStateNode_3 = MoveLoopRandom() {is=title AnimStateNode, stateEntered {guid=#7, notifyName=MoveLoopRandomEntered}}
   AnimStateTransitionNode_0 = AnimStateTransitionNode()
   AnimStateTransitionNode_1 = AnimStateTransitionNode()
   AnimStateTransitionNode_10 = AnimStateTransitionNode()
   AnimStateTransitionNode_2 = AnimStateTransitionNode() {bAutomaticRuleBasedOnSequencePlayerInState=true, crossfadeDuration=0.4}
   AnimStateTransitionNode_3 = AnimStateTransitionNode()
   AnimStateTransitionNode_4 = AnimStateTransitionNode()
   AnimStateTransitionNode_5 = AnimStateTransitionNode()
   AnimStateTransitionNode_6 = AnimStateTransitionNode()
   AnimStateTransitionNode_7 = AnimStateTransitionNode() {bAutomaticRuleBasedOnSequencePlayerInState=true, crossfadeDuration=0}
   AnimStateTransitionNode_8 = AnimStateTransitionNode()
   AnimStateTransitionNode_9 = AnimStateTransitionNode()
  ]
  links [
   {
    dst=MoveLoopRandom
    rule=AnimGraph.AnimGraphNode_StateMachine_0.Default.AnimStateNode_0.Locomotion.AnimGraphNode_StateMachine_0.Locomotion.AnimStateNode_2.MoveLoop.AnimGraphNode_StateMachine_0.MoveLoopSM.AnimStateTransitionNode_6.Transition
    src=MoveLoop
    via=AnimStateTransitionNode_6
   }
   {
    dst=MoveLoopUnderWater
    rule=AnimGraph.AnimGraphNode_StateMachine_0.Default.AnimStateNode_0.Locomotion.AnimGraphNode_StateMachine_0.Locomotion.AnimStateNode_2.MoveLoop.AnimGraphNode_StateMachine_0.MoveLoopSM.AnimStateTransitionNode_0.Transition
    src=MoveLoop
    via=AnimStateTransitionNode_0
   }
   {
    dst=MoveLoop
    rule=AnimGraph.AnimGraphNode_StateMachine_0.Default.AnimStateNode_0.Locomotion.AnimGraphNode_StateMachine_0.Locomotion.AnimStateNode_2.MoveLoop.AnimGraphNode_StateMachine_0.MoveLoopSM.AnimStateTransitionNode_10.Transition
    src=MoveLoopRandom
    via=AnimStateTransitionNode_10
   }
   {
    dst=MoveLoop
    rule=AnimGraph.AnimGraphNode_StateMachine_0.Default.AnimStateNode_0.Locomotion.AnimGraphNode_StateMachine_0.Locomotion.AnimStateNode_2.MoveLoop.AnimGraphNode_StateMachine_0.MoveLoopSM.AnimStateTransitionNode_7.Transition
    src=MoveLoopRandom
    via=AnimStateTransitionNode_7
   }
   {
    dst=MoveLoop
    rule=AnimGraph.AnimGraphNode_StateMachine_0.Default.AnimStateNode_0.Locomotion.AnimGraphNode_StateMachine_0.Locomotion.AnimStateNode_2.MoveLoop.AnimGraphNode_StateMachine_0.MoveLoopSM.AnimStateTransitionNode_8.Transition
    src=MoveLoopRandom
    via=AnimStateTransitionNode_8
   }
   {
    dst=MoveLoop
    rule=AnimGraph.AnimGraphNode_StateMachine_0.Default.AnimStateNode_0.Locomotion.AnimGraphNode_StateMachine_0.Locomotion.AnimStateNode_2.MoveLoop.AnimGraphNode_StateMachine_0.MoveLoopSM.AnimStateTransitionNode_9.Transition
    src=MoveLoopRandom
    via=AnimStateTransitionNode_9
   }
   {
    dst=MoveLoop
    rule=AnimGraph.AnimGraphNode_StateMachine_0.Default.AnimStateNode_0.Locomotion.AnimGraphNode_StateMachine_0.Locomotion.AnimStateNode_2.MoveLoop.AnimGraphNode_StateMachine_0.MoveLoopSM.AnimStateTransitionNode_4.Transition
    src=MoveLoopUnderWater
    via=AnimStateTransitionNode_4
   }
   {
    dst=MoveLoopUnderWaterExit
    rule=AnimGraph.AnimGraphNode_StateMachine_0.Default.AnimStateNode_0.Locomotion.AnimGraphNode_StateMachine_0.Locomotion.AnimStateNode_2.MoveLoop.AnimGraphNode_StateMachine_0.MoveLoopSM.AnimStateTransitionNode_1.Transition
    src=MoveLoopUnderWater
    via=AnimStateTransitionNode_1
   }
   {
    dst=MoveLoop
    rule=AnimGraph.AnimGraphNode_StateMachine_0.Default.AnimStateNode_0.Locomotion.AnimGraphNode_StateMachine_0.Locomotion.AnimStateNode_2.MoveLoop.AnimGraphNode_StateMachine_0.MoveLoopSM.AnimStateTransitionNode_2.Transition
    src=MoveLoopUnderWaterExit
    via=AnimStateTransitionNode_2
   }
   {
    dst=MoveLoop
    rule=AnimGraph.AnimGraphNode_StateMachine_0.Default.AnimStateNode_0.Locomotion.AnimGraphNode_StateMachine_0.Locomotion.AnimStateNode_2.MoveLoop.AnimGraphNode_StateMachine_0.MoveLoopSM.AnimStateTransitionNode_5.Transition
    src=MoveLoopUnderWaterExit
    via=AnimStateTransitionNode_5
   }
   {
    dst=MoveLoopUnderWater
    rule=AnimGraph.AnimGraphNode_StateMachine_0.Default.AnimStateNode_0.Locomotion.AnimGraphNode_StateMachine_0.Locomotion.AnimStateNode_2.MoveLoop.AnimGraphNode_StateMachine_0.MoveLoopSM.AnimStateTransitionNode_3.Transition
    src=MoveLoopUnderWaterExit
    via=AnimStateTransitionNode_3
   }
  ]
 }
 MoveLoop {
  path=AnimGraph.AnimGraphNode_StateMachine_0.Default.AnimStateNode_0.Locomotion.AnimGraphNode_StateMachine_0.Locomotion.AnimStateNode_2.MoveLoop.AnimGraphNode_StateMachine_0.MoveLoopSM.AnimStateNode_0.MoveLoop
  nodes [
   n1 = Sequence Player 'Locomotion_FlightLoop_Default'() {
    groupName=IdleSync
    method=SyncGroup
    playRate=1.5
    sequence=/Game/Animation/Game/Pets/HightowerRadish/Backpack/Locomotion/Locomotion_FlightLoop_Default
   }
   n2 = "Sequence Player (additive) 'Locomotion_FlightLoop_Crouch_Pose'"() {bIgnoreForRelevancyTest=true, sequence=&Locomotion_FlightLoop_Crouch_Pose}
   n3 = Apply Additive(
    Base=Apply Mesh Space Additive(Base=@n1, Additive=@n2, Alpha=1) {alphaBoolBlend {blendOption=Linear}}
    Additive=Blendspace Player 'Locomotion_LeanPoses_BS'(X=Get OwnerLocomotionDirection() -> (OwnerLocomotionDirection:float), Y=Get OwnerLocomotionSpeed() -> (OwnerLocomotionSpeed:float)) {bIgnoreForRelevancyTest=true, blendSpace=/Game/Animation/Game/Pets/HightowerRadish/Backpack/Locomotion/Locomotion_LeanPoses_BS}
   ) {alphaBoolBlend {blendOption=Linear}}
   n4 = Blendspace Player 'Locomotion_Crouch_PositionOffsetPoses_BS'(X=Get OwnerLocomotionDirection(), Y=Get OwnerLocomotionSpeed()) {
    bIgnoreForRelevancyTest=true
    blendSpace=/Game/Animation/Game/Pets/HightowerRadish/Backpack/Locomotion/Locomotion_Crouch_PositionOffsetPoses_BS
   }
   n5 = Get OwnerLocomotionSpeed()
   n6 = Get OwnerLocomotionDirection()
   n7 = Sequence Player 'Locomotion_FlightLoop_Default'() {
    groupName=IdleSync
    method=SyncGroup
    playRate=1.5
    sequence=&Locomotion_FlightLoop_Default
   }
   n8 = Apply Additive(
    Base=Apply Additive(
     Base=@n7
     Additive=Blendspace Player 'Locomotion_LeanPoses_BS'(X=@n6, Y=@n5) {bIgnoreForRelevancyTest=true, blendSpace=&Locomotion_LeanPoses_BS}
    ) {alphaBoolBlend {blendOption=Linear}}
    Additive=Blend Poses by bool(
     BlendPose_0="Sequence Player (additive) 'Locomotion_ChangeDirectionPose'"() {bIgnoreForRelevancyTest=true, sequence=/Game/Animation/Game/Pets/HightowerRadish/Backpack/Locomotion/Locomotion_ChangeDirectionPose}
     BlendPose_1=Additive Identity Pose()
     BlendTime_0=0.2
     BlendTime_1=0.6
     bActiveValue=Get bSharpDirectionChange() -> (bSharpDirectionChange:bool)
    ) {blendType=Cubic}
   ) {alphaBoolBlend {blendOption=Linear}}
   n9 = Sequence Player 'Hover_Idle_Downsights'() {sequence=&Hover_Idle_Downsights}
   n10 = Get bHasTargettedInLastNSeconds() -> (bHasTargettedInLastNSeconds:bool)
   Output Animation Pose(Result=Blend Poses by bool(
     BlendPose_0=Apply Additive(Base=@n3, Additive=@n4) {alphaBoolBlend {blendOption=Linear}}
     BlendPose_1=Blend Poses by bool(BlendPose_0=@n9, BlendPose_1=@n8, BlendTime_0=0.4, BlendTime_1=0.4, bActiveValue=@n10) {blendType=Linear}
     BlendTime_0=0.4
     BlendTime_1=0.4
     bActiveValue=Get bIsCrouching()
    ) {blendType=Linear})
   Get bIsTargeting()
  ]
 }
 MoveLoopUnderWater {
  path=AnimGraph.AnimGraphNode_StateMachine_0.Default.AnimStateNode_0.Locomotion.AnimGraphNode_StateMachine_0.Locomotion.AnimStateNode_2.MoveLoop.AnimGraphNode_StateMachine_0.MoveLoopSM.AnimStateNode_1.MoveLoopUnderWater
  nodes [
   n1 = "Sequence Player (additive) 'Hover_Idle_Underwater_Enter_Tail_Add'"(PlayRate=2, bLoopAnimation=False) {bLoopAnimation=false, playRate=2, sequence=&Hover_Idle_Underwater_Enter_Tail_Add}
   n2 = Sequence Player 'Locomotion_UnderWater'() {groupName=IdleSync, method=SyncGroup, sequence=/Game/Animation/Game/Pets/HightowerRadish/Backpack/Locomotion/Locomotion_UnderWater}
   Output Animation Pose(Result=Apply Additive(Base=@n2, Additive=@n1) {alphaBoolBlend {blendOption=Linear}})
  ]
 }
 MoveLoopUnderWaterExit {
  path=AnimGraph.AnimGraphNode_StateMachine_0.Default.AnimStateNode_0.Locomotion.AnimGraphNode_StateMachine_0.Locomotion.AnimStateNode_2.MoveLoop.AnimGraphNode_StateMachine_0.MoveLoopSM.AnimStateNode_2.MoveLoopUnderWaterExit
  nodes [
   n1 = Get OwnerLocomotionDirection()
   n2 = Get OwnerLocomotionSpeed()
   Output Animation Pose(Result=Apply Additive(
     Base=Sequence Player 'Locomotion_UnderWater_Exit'!(bLoopAnimation:bool=False) -> (Pose) {bLoopAnimation=false, sequence=/Game/Animation/Game/Pets/HightowerRadish/Backpack/Locomotion/Locomotion_UnderWater_Exit}
     Additive=Blendspace Player 'Locomotion_LeanPoses_BS'(X=@n1, Y=@n2) {blendSpace=&Locomotion_LeanPoses_BS}
    ) {alphaBoolBlend {blendOption=Linear}})
  ]
 }
 MoveLoopRandom {
  path=AnimGraph.AnimGraphNode_StateMachine_0.Default.AnimStateNode_0.Locomotion.AnimGraphNode_StateMachine_0.Locomotion.AnimStateNode_2.MoveLoop.AnimGraphNode_StateMachine_0.MoveLoopSM.AnimStateNode_3.MoveLoopRandom
  nodes [
   n1 = Get OwnerLocomotionDirection()
   n2 = Get OwnerLocomotionSpeed()
   n3 = Get OwnerLocomotionSpeed()
   n4 = Get OwnerLocomotionDirection()
   n5 = Blend Poses by int!(
    BlendPose_0:Pose Link Structure=Apply Additive(
     Base=Sequence Player 'Locomotion_FlightLoop_Break_01'() {bLoopAnimation=false, sequence=/Game/Animation/Game/Pets/HightowerRadish/Backpack/Locomotion/Locomotion_FlightLoop_Break_01}
     Additive=Blendspace Player 'Locomotion_LeanPoses_BS'(X=@n1, Y=@n2) {bIgnoreForRelevancyTest=true, blendSpace=&Locomotion_LeanPoses_BS}
     Alpha=Get TurnOffLeans() -> (TurnOffLeans:float)
    ) {alphaBoolBlend {blendOption=Linear}}
    BlendPose_1:Pose Link Structure=Apply Additive(
     Base=Sequence Player 'Locomotion_FlightLoop_Break_02'() {bLoopAnimation=false, sequence=/Game/Animation/Game/Pets/HightowerRadish/Backpack/Locomotion/Locomotion_FlightLoop_Break_02}
     Additive=Blendspace Player 'Locomotion_LeanPoses_BS'(X=@n4, Y=@n3) {bIgnoreForRelevancyTest=true, blendSpace=&Locomotion_LeanPoses_BS}
     Alpha=Get TurnOffLeans()
    ) {alphaBoolBlend {blendOption=Linear}}
    BlendTime_0:float=0.1
    BlendTime_1:float=0.1
    ActiveChildIndex:int=Get RandomLocomotionBreak() -> (RandomLocomotionBreak:int=0)
   ) -> (Pose:Pose Link Structure) {
    is=title AnimGraphNode_BlendListByInt
    blendType=Linear
    blendPose [{linkId=-1, sourceLinkId=-1}, {linkId=-1, sourceLinkId=-1}]
    blendTime [0.1,0.1]
   }
   Output Animation Pose(Result=@n5)
  ]
 }
 Transition {
  path=AnimGraph.AnimGraphNode_StateMachine_0.Default.AnimStateNode_0.Locomotion.AnimGraphNode_StateMachine_0.Locomotion.AnimStateNode_2.MoveLoop.AnimGraphNode_StateMachine_0.MoveLoopSM.AnimStateTransitionNode_0.Transition
  nodes [Result(bCanEnterTransition=Get bPetUnderWaterAndNotTargeting())]
 }
 Transition {
  path=AnimGraph.AnimGraphNode_StateMachine_0.Default.AnimStateNode_0.Locomotion.AnimGraphNode_StateMachine_0.Locomotion.AnimStateNode_2.MoveLoop.AnimGraphNode_StateMachine_0.MoveLoopSM.AnimStateTransitionNode_1.Transition
  nodes [Result(bCanEnterTransition=NOT Boolean(A=Get bPetUnderwater()))]
 }
 Transition {
  path=AnimGraph.AnimGraphNode_StateMachine_0.Default.AnimStateNode_0.Locomotion.AnimGraphNode_StateMachine_0.Locomotion.AnimStateNode_2.MoveLoop.AnimGraphNode_StateMachine_0.MoveLoopSM.AnimStateTransitionNode_10.Transition
  nodes [Result(bCanEnterTransition=Get bIsCrouching())]
 }
 Transition {
  path=AnimGraph.AnimGraphNode_StateMachine_0.Default.AnimStateNode_0.Locomotion.AnimGraphNode_StateMachine_0.Locomotion.AnimStateNode_2.MoveLoop.AnimGraphNode_StateMachine_0.MoveLoopSM.AnimStateTransitionNode_2.Transition
  nodes [Result()]
 }
 Transition {
  path=AnimGraph.AnimGraphNode_StateMachine_0.Default.AnimStateNode_0.Locomotion.AnimGraphNode_StateMachine_0.Locomotion.AnimStateNode_2.MoveLoop.AnimGraphNode_StateMachine_0.MoveLoopSM.AnimStateTransitionNode_3.Transition
  nodes [Result(bCanEnterTransition=Get bPetUnderWaterAndNotTargeting())]
 }
 Transition {
  path=AnimGraph.AnimGraphNode_StateMachine_0.Default.AnimStateNode_0.Locomotion.AnimGraphNode_StateMachine_0.Locomotion.AnimStateNode_2.MoveLoop.AnimGraphNode_StateMachine_0.MoveLoopSM.AnimStateTransitionNode_4.Transition
  nodes [Result(bCanEnterTransition=Get bIsTargeting())]
 }
 Transition {
  path=AnimGraph.AnimGraphNode_StateMachine_0.Default.AnimStateNode_0.Locomotion.AnimGraphNode_StateMachine_0.Locomotion.AnimStateNode_2.MoveLoop.AnimGraphNode_StateMachine_0.MoveLoopSM.AnimStateTransitionNode_5.Transition
  nodes [Result(bCanEnterTransition=Get bIsTargeting())]
 }
 Transition {
  path=AnimGraph.AnimGraphNode_StateMachine_0.Default.AnimStateNode_0.Locomotion.AnimGraphNode_StateMachine_0.Locomotion.AnimStateNode_2.MoveLoop.AnimGraphNode_StateMachine_0.MoveLoopSM.AnimStateTransitionNode_6.Transition
  nodes [
   n1 = "Get Relevant Anim Time Remaining (MoveLoop)"!() -> (ReturnValue:float) {is=title K2Node_AnimGetter}
   n2 = Get bIsCrouching()
   n3 = Get bIsPlayingMeleeAnim!(self:Fort Player Anim Instance Object Reference=Get OwnerAnimBP()) -> (bIsPlayingMeleeAnim)
   Result(bCanEnterTransition=AND Boolean!(
     A=Get bAllowRandomLocomotionBreak() -> (bAllowRandomLocomotionBreak:bool)
     B=NOT Boolean(A=@n3)
     C:bool="float <= float"!(A:double=@n1, B:double=0.2) -> (ReturnValue:bool) {is=title K2Node_CallFunction}
     D:bool=NOT Boolean(A=@n2)
    ) -> (ReturnValue))
  ]
 }
 Transition {
  path=AnimGraph.AnimGraphNode_StateMachine_0.Default.AnimStateNode_0.Locomotion.AnimGraphNode_StateMachine_0.Locomotion.AnimStateNode_2.MoveLoop.AnimGraphNode_StateMachine_0.MoveLoopSM.AnimStateTransitionNode_7.Transition
  nodes [Result()]
 }
 Transition {
  path=AnimGraph.AnimGraphNode_StateMachine_0.Default.AnimStateNode_0.Locomotion.AnimGraphNode_StateMachine_0.Locomotion.AnimStateNode_2.MoveLoop.AnimGraphNode_StateMachine_0.MoveLoopSM.AnimStateTransitionNode_8.Transition
  nodes [Result(bCanEnterTransition=Get bPetUnderwater())]
 }
 Transition {
  path=AnimGraph.AnimGraphNode_StateMachine_0.Default.AnimStateNode_0.Locomotion.AnimGraphNode_StateMachine_0.Locomotion.AnimStateNode_2.MoveLoop.AnimGraphNode_StateMachine_0.MoveLoopSM.AnimStateTransitionNode_9.Transition
  nodes [Result(bCanEnterTransition=Get bIsTargeting())]
 }
 MoveStop {
  path=AnimGraph.AnimGraphNode_StateMachine_0.Default.AnimStateNode_0.Locomotion.AnimGraphNode_StateMachine_0.Locomotion.AnimStateNode_3.MoveStop
  nodes [
   Apply Additive() {alphaBoolBlend {blendOption=Linear}}
   n1 = Sequence Player 'Locomotion_FlightLoop_Exit'!(StartPosition:float=0.2) -> (Pose) {
    bLoopAnimation=false
    groupName=Flightloop_Exit
    method=SyncGroup
    sequence=/Game/Animation/Game/Pets/HightowerRadish/Backpack/Locomotion/Locomotion_FlightLoop_Exit
   }
   n2 = "Sequence Player (additive) 'Locomotion_FlightLoop_Crouch_Pose'"() {bIgnoreForRelevancyTest=true, sequence=&Locomotion_FlightLoop_Crouch_Pose}
   n3 = Two Way Blend(
    A=@n2
    B="Sequence Player (additive) 'Hover_Idle_Crouch_Pose'"() {bIgnoreForRelevancyTest=true, sequence=&Hover_Idle_Crouch_Pose}
    Alpha=Get AdditiveTransitionCurve() -> (AdditiveTransitionCurve:float)
   ) {blendNode {alphaBoolBlend {blendOption=Linear}}}
   n4 = Sequence Player 'Locomotion_FlightLoop_Exit'(StartPosition=0.2) {
    bLoopAnimation=false
    groupName=Flightloop_Exit
    method=SyncGroup
    sequence=&Locomotion_FlightLoop_Exit
   }
   Output Animation Pose(Result=Blend Poses by bool(
     BlendPose_0=Apply Mesh Space Additive(Base=@n1, Additive=@n3, Alpha=1) {alphaBoolBlend {blendOption=Linear}}
     BlendPose_1=@n4
     BlendTime_0=0.25
     BlendTime_1=0.25
     bActiveValue=Get bIsCrouching()
    ) {blendType=Linear})
  ]
 }
 Transition {
  path=AnimGraph.AnimGraphNode_StateMachine_0.Default.AnimStateNode_0.Locomotion.AnimGraphNode_StateMachine_0.Locomotion.AnimStateTransitionNode_0.Transition
  nodes [Result()]
 }
 Transition {
  path=AnimGraph.AnimGraphNode_StateMachine_0.Default.AnimStateNode_0.Locomotion.AnimGraphNode_StateMachine_0.Locomotion.AnimStateTransitionNode_1.Transition
  nodes [Result(bCanEnterTransition=Get bIsMoving2D())]
 }
 Transition {
  path=AnimGraph.AnimGraphNode_StateMachine_0.Default.AnimStateNode_0.Locomotion.AnimGraphNode_StateMachine_0.Locomotion.AnimStateTransitionNode_10.Transition
  nodes [
   Result(bCanEnterTransition=Get bIsTargeting())
   Get bIsTargeting!(self:Fort Player Anim Instance Object Reference=Get OwnerAnimBP()) -> (bIsTargeting)
  ]
 }
 Transition {
  path=AnimGraph.AnimGraphNode_StateMachine_0.Default.AnimStateNode_0.Locomotion.AnimGraphNode_StateMachine_0.Locomotion.AnimStateTransitionNode_2.Transition
  nodes [
   n1 = Get bStoppedMovingOrMeleeForLongEnoughTime() -> (bStoppedMovingOrMeleeForLongEnoughTime:bool)
   Result(bCanEnterTransition=@n1)
   AND Boolean!(A=@n1, B=NOT Boolean(A=Get bHasTargettedInLastNSeconds())) -> (ReturnValue)
  ]
 }
 Transition {
  path=AnimGraph.AnimGraphNode_StateMachine_0.Default.AnimStateNode_0.Locomotion.AnimGraphNode_StateMachine_0.Locomotion.AnimStateTransitionNode_4.Transition
  nodes [
   n1 = Get bStoppedMovingOrMeleeForLongEnoughTime()
   Result(bCanEnterTransition=@n1)
   AND Boolean(A=@n1, B=NOT Boolean(A=Get bHasTargettedInLastNSeconds()))
   Get bIsTargeting!() -> (bIsTargeting)
  ]
 }
 Transition {
  path=AnimGraph.AnimGraphNode_StateMachine_0.Default.AnimStateNode_0.Locomotion.AnimGraphNode_StateMachine_0.Locomotion.AnimStateTransitionNode_5.Transition
  nodes [Result()]
 }
 Transition {
  path=AnimGraph.AnimGraphNode_StateMachine_0.Default.AnimStateNode_0.Locomotion.AnimGraphNode_StateMachine_0.Locomotion.AnimStateTransitionNode_8.Transition
  nodes [Result(bCanEnterTransition=Get bIsOwnerPlayingMeleeAnim() -> (bIsOwnerPlayingMeleeAnim:bool))]
 }
 Transition {
  path=AnimGraph.AnimGraphNode_StateMachine_0.Default.AnimStateNode_0.Locomotion.AnimGraphNode_StateMachine_0.Locomotion.AnimStateTransitionNode_9.Transition
  nodes [Result(bCanEnterTransition=Get bIsOwnerPlayingMeleeAnim())]
 }
 EmoteReactIntro {
  path=AnimGraph.AnimGraphNode_StateMachine_0.Default.AnimStateNode_1.EmoteReactIntro
  nodes [
   Sequence Player 'Proto_Emote_React_Intro'() {bLoopAnimation=false, sequence=/Game/Animation/Game/Pets/HightowerRadish/Proto/Proto_Emote_React_Intro}
   Output Animation Pose(Result=Sequence Player 'Emote_React_Intro'!(bLoopAnimation:bool=False) -> (Pose) {bLoopAnimation=false, sequence=/Game/Animation/Game/Pets/HightowerRadish/Backpack/Emote/Emote_React_Intro})
  ]
 }
 EmoteReactLoop {
  path=AnimGraph.AnimGraphNode_StateMachine_0.Default.AnimStateNode_2.EmoteReactLoop
  nodes [
   Sequence Player 'Proto_Emote_React_Loop'() {sequence=/Game/Animation/Game/Pets/HightowerRadish/Proto/Proto_Emote_React_Loop}
   Output Animation Pose(Result=Sequence Player 'Emote_React_Loop'() {sequence=/Game/Animation/Game/Pets/HightowerRadish/Backpack/Emote/Emote_React_Loop})
  ]
 }
 Skydive {
  path=AnimGraph.AnimGraphNode_StateMachine_0.Default.AnimStateNode_3.Skydive
  nodes [
   Blend Poses by bool(
    BlendPose_0=Sequence Player 'Paraglide'() {sequence=/Game/Animation/Game/Pets/HightowerRadish/Backpack/Glider/Paraglide}
    BlendPose_1="Sequence Player (additive) 'Glide'"() {sequence=/Game/Animation/Game/Pets/HightowerRadish/Backpack/Glider/Glide}
    BlendTime_0=0.4
    BlendTime_1=0.4
    bActiveValue=Break Fort Anim Input Skydiving!(FortAnimInput_Skydiving:"Fort Anim Input Skydiving Structure (by ref)"=Get SkydivingState() -> (SkydivingState:Fort Anim Input Skydiving Structure)) -> (
     ParaGilderRootOffset:Vector=~
     ParaGliderRootRotationOverride:Rotator=~
     bUseParaGlideRootModifier:bool=~
     bIsSkydivingFromLaunchPad:bool=~
     bIsSkydivingFromBus:bool=~
     bIsInVortex:bool=~
     bIsDiving:bool=~
     bIsParachuteOpen:bool=~
     bIsSkydiving:bool=~
     bIsParachuteLeaning:bool=~
     bIsSkydiveLeaning:bool=~
     bIsLeaning:bool=~
     bIsSkydiveDiveMode:bool=~
     bPlaySkydiveDrift:bool=~
     bSkydiveDriftDelayActive:bool=~
     bSkydiveDriftAnimAllowed:bool=~
     bIsGliderRight:bool=~
     bIsGliderCenter:bool=~
     bIsGliderLeft:bool=~
     bIsGliderForward:bool=~
     bIsGliderBack:bool=~
     bHasDeployedAndIsStandingUpright:bool=~
     LocalAccelForward:float=~
     LocalAccelRight:float=~
     LocalVelocityRight:float=~
     SkydiveAimPitch:float=~
     SkydiveAimYaw:float=~
     DeployChuteAnimRate:float=~
     SkydiveDriftAnimRate:float=~
     SkydiveDriftAnimRateCurrent:float=~
     SkydiveFidgetAnimRate:float=~
     SkydiveFidgetAnimRateCurrent:float=~
     SkydiveAdditiveAlpha:float=~
     SkydiveDriftDelay:float=~
     SkydiveDriftAnim:int
     SkydiveDriftAnimMax:int
     LaunchpadAnim:int
     LocalAccelDir:ESkydivingDirection Enum
     DirectionLast:ESkydivingDirection Enum
     PoseAdditive:Anim Sequence Object Reference
    ) {is=title K2Node_BreakStruct}.bIsParachuteOpen
   ) {blendType=Linear}
   AnimGraphNode_StateMachine_0 = SkyDive() {is=title AnimGraphNode_StateMachine}
   Output Animation Pose(Result=@AnimGraphNode_StateMachine_0)
  ]
 }
 SkyDive {
  entry=Glide
  path=AnimGraph.AnimGraphNode_StateMachine_0.Default.AnimStateNode_3.Skydive.AnimGraphNode_StateMachine_0.SkyDive
  ids=id
  nodes [
   AnimStateNode_0 = Glide() {is=title AnimStateNode}
   AnimStateNode_1 = GlideToParaglide() {is=title AnimStateNode}
   AnimStateNode_2 = Paraglide() {is=title AnimStateNode}
   AnimStateNode_3 = ParaglideToGlide() {is=title AnimStateNode}
   AnimStateNode_4 = ToDive() {is=title AnimStateNode}
   AnimStateNode_5 = DiveLoop() {is=title AnimStateNode}
   AnimStateNode_6 = DiveToGlide() {is=title AnimStateNode}
   AnimStateTransitionNode_0 = AnimStateTransitionNode()
   AnimStateTransitionNode_1 = AnimStateTransitionNode() {bAutomaticRuleBasedOnSequencePlayerInState=true, crossfadeDuration=0}
   AnimStateTransitionNode_2 = AnimStateTransitionNode()
   AnimStateTransitionNode_3 = AnimStateTransitionNode() {bAutomaticRuleBasedOnSequencePlayerInState=true, crossfadeDuration=0.4}
   AnimStateTransitionNode_4 = AnimStateTransitionNode()
   AnimStateTransitionNode_5 = AnimStateTransitionNode() {bAutomaticRuleBasedOnSequencePlayerInState=true}
   AnimStateTransitionNode_6 = AnimStateTransitionNode()
   AnimStateTransitionNode_8 = AnimStateTransitionNode()
   AnimStateTransitionNode_9 = AnimStateTransitionNode() {bAutomaticRuleBasedOnSequencePlayerInState=true, crossfadeDuration=0.1}
  ]
  links [
   {
    dst=DiveToGlide
    rule=AnimGraph.AnimGraphNode_StateMachine_0.Default.AnimStateNode_3.Skydive.AnimGraphNode_StateMachine_0.SkyDive.AnimStateTransitionNode_8.Transition
    src=DiveLoop
    via=AnimStateTransitionNode_8
   }
   {
    dst=Glide
    rule=AnimGraph.AnimGraphNode_StateMachine_0.Default.AnimStateNode_3.Skydive.AnimGraphNode_StateMachine_0.SkyDive.AnimStateTransitionNode_9.Transition
    src=DiveToGlide
    via=AnimStateTransitionNode_9
   }
   {
    dst=GlideToParaglide
    rule=AnimGraph.AnimGraphNode_StateMachine_0.Default.AnimStateNode_3.Skydive.AnimGraphNode_StateMachine_0.SkyDive.AnimStateTransitionNode_0.Transition
    src=Glide
    via=AnimStateTransitionNode_0
   }
   {
    dst=ToDive
    rule=AnimGraph.AnimGraphNode_StateMachine_0.Default.AnimStateNode_3.Skydive.AnimGraphNode_StateMachine_0.SkyDive.AnimStateTransitionNode_4.Transition
    src=Glide
    via=AnimStateTransitionNode_4
   }
   {
    dst=Paraglide
    rule=AnimGraph.AnimGraphNode_StateMachine_0.Default.AnimStateNode_3.Skydive.AnimGraphNode_StateMachine_0.SkyDive.AnimStateTransitionNode_1.Transition
    src=GlideToParaglide
    via=AnimStateTransitionNode_1
   }
   {
    dst=ParaglideToGlide
    rule=AnimGraph.AnimGraphNode_StateMachine_0.Default.AnimStateNode_3.Skydive.AnimGraphNode_StateMachine_0.SkyDive.AnimStateTransitionNode_2.Transition
    src=Paraglide
    via=AnimStateTransitionNode_2
   }
   {
    dst=Glide
    rule=AnimGraph.AnimGraphNode_StateMachine_0.Default.AnimStateNode_3.Skydive.AnimGraphNode_StateMachine_0.SkyDive.AnimStateTransitionNode_3.Transition
    src=ParaglideToGlide
    via=AnimStateTransitionNode_3
   }
   {
    dst=DiveLoop
    rule=AnimGraph.AnimGraphNode_StateMachine_0.Default.AnimStateNode_3.Skydive.AnimGraphNode_StateMachine_0.SkyDive.AnimStateTransitionNode_5.Transition
    src=ToDive
    via=AnimStateTransitionNode_5
   }
   {
    dst=DiveLoop
    rule=AnimGraph.AnimGraphNode_StateMachine_0.Default.AnimStateNode_3.Skydive.AnimGraphNode_StateMachine_0.SkyDive.AnimStateTransitionNode_6.Transition
    src=ToDive
    via=AnimStateTransitionNode_6
   }
  ]
 }
 Glide {
  path=AnimGraph.AnimGraphNode_StateMachine_0.Default.AnimStateNode_3.Skydive.AnimGraphNode_StateMachine_0.SkyDive.AnimStateNode_0.Glide
  nodes [
   n1 = Break Fort Anim Input Skydiving(FortAnimInput_Skydiving=Get SkydivingState()) -> (
    ParaGilderRootOffset=~
    ParaGliderRootRotationOverride=~
    bUseParaGlideRootModifier=~
    bIsSkydivingFromLaunchPad=~
    bIsSkydivingFromBus=~
    bIsInVortex=~
    bIsDiving=~
    bIsParachuteOpen=~
    bIsSkydiving=~
    bIsParachuteLeaning=~
    bIsSkydiveLeaning=~
    bIsLeaning=~
    bIsSkydiveDiveMode=~
    bPlaySkydiveDrift=~
    bSkydiveDriftDelayActive=~
    bSkydiveDriftAnimAllowed=~
    bIsGliderRight=~
    bIsGliderCenter=~
    bIsGliderLeft=~
    bIsGliderForward=~
    bIsGliderBack=~
    bHasDeployedAndIsStandingUpright=~
    LocalAccelForward=~
    LocalAccelRight=~
    LocalVelocityRight=~
    SkydiveAimPitch=~
    SkydiveAimYaw=~
    DeployChuteAnimRate=~
    SkydiveDriftAnimRate=~
    SkydiveDriftAnimRateCurrent=~
    SkydiveFidgetAnimRate=~
    SkydiveFidgetAnimRateCurrent=~
    SkydiveAdditiveAlpha=~
    SkydiveDriftDelay=~
   )
   Output Animation Pose(Result=Apply Additive(
     Base=Blendspace Player 'BS_Glide_Lean'(X=@n1.LocalVelocityRight, Y=@n1.SkydiveAimPitch) {blendSpace=/Game/Animation/Game/Pets/HightowerRadish/Backpack/Glider/BS_Glide_Lean}
     Additive="Sequence Player (additive) 'Glide'"() {sequence=&Glide}
    ) {alphaBoolBlend {blendOption=Linear}})
  ]
 }
 GlideToParaglide {
  path=AnimGraph.AnimGraphNode_StateMachine_0.Default.AnimStateNode_3.Skydive.AnimGraphNode_StateMachine_0.SkyDive.AnimStateNode_1.GlideToParaglide
  nodes [Output Animation Pose(Result=Sequence Player 'Glide_To_Paraglide'() {bLoopAnimation=false, sequence=/Game/Animation/Game/Pets/HightowerRadish/Backpack/Glider/Glide_To_Paraglide})]
 }
 Paraglide {
  path=AnimGraph.AnimGraphNode_StateMachine_0.Default.AnimStateNode_3.Skydive.AnimGraphNode_StateMachine_0.SkyDive.AnimStateNode_2.Paraglide
  nodes [Output Animation Pose(Result=Sequence Player 'Paraglide'() {sequence=&Paraglide})]
 }
 ParaglideToGlide {
  path=AnimGraph.AnimGraphNode_StateMachine_0.Default.AnimStateNode_3.Skydive.AnimGraphNode_StateMachine_0.SkyDive.AnimStateNode_3.ParaglideToGlide
  nodes [Output Animation Pose(Result=Sequence Player 'Paraglide_To_Glide'() {bLoopAnimation=false, sequence=/Game/Animation/Game/Pets/HightowerRadish/Backpack/Glider/Paraglide_To_Glide})]
 }
 ToDive {
  path=AnimGraph.AnimGraphNode_StateMachine_0.Default.AnimStateNode_3.Skydive.AnimGraphNode_StateMachine_0.SkyDive.AnimStateNode_4.ToDive
  nodes [Output Animation Pose(Result=Sequence Player 'Glide_To_Dive'() {sequence=/Game/Animation/Game/Pets/HightowerRadish/Backpack/Glider/Glide_To_Dive})]
 }
 DiveLoop {
  path=AnimGraph.AnimGraphNode_StateMachine_0.Default.AnimStateNode_3.Skydive.AnimGraphNode_StateMachine_0.SkyDive.AnimStateNode_5.DiveLoop
  nodes [Output Animation Pose(Result=Sequence Player 'DiveLoop'() {sequence=/Game/Animation/Game/Pets/HightowerRadish/Backpack/Glider/DiveLoop})]
 }
 DiveToGlide {
  path=AnimGraph.AnimGraphNode_StateMachine_0.Default.AnimStateNode_3.Skydive.AnimGraphNode_StateMachine_0.SkyDive.AnimStateNode_6.DiveToGlide
  nodes [Output Animation Pose(Result=Sequence Player 'Dive_To_Glide'() {sequence=/Game/Animation/Game/Pets/HightowerRadish/Backpack/Glider/Dive_To_Glide})]
 }
 Transition {
  path=AnimGraph.AnimGraphNode_StateMachine_0.Default.AnimStateNode_3.Skydive.AnimGraphNode_StateMachine_0.SkyDive.AnimStateTransitionNode_0.Transition
  nodes [Result(bCanEnterTransition=Break Fort Anim Input Skydiving(FortAnimInput_Skydiving=Get SkydivingState()) -> (
     ParaGilderRootOffset=~
     ParaGliderRootRotationOverride=~
     bUseParaGlideRootModifier=~
     bIsSkydivingFromLaunchPad=~
     bIsSkydivingFromBus=~
     bIsInVortex=~
     bIsDiving=~
     bIsParachuteOpen=~
     bIsSkydiving=~
     bIsParachuteLeaning=~
     bIsSkydiveLeaning=~
     bIsLeaning=~
     bIsSkydiveDiveMode=~
     bPlaySkydiveDrift=~
     bSkydiveDriftDelayActive=~
     bSkydiveDriftAnimAllowed=~
     bIsGliderRight=~
     bIsGliderCenter=~
     bIsGliderLeft=~
     bIsGliderForward=~
     bIsGliderBack=~
     bHasDeployedAndIsStandingUpright=~
     LocalAccelForward=~
     LocalAccelRight=~
     LocalVelocityRight=~
     SkydiveAimPitch=~
     SkydiveAimYaw=~
     DeployChuteAnimRate=~
     SkydiveDriftAnimRate=~
     SkydiveDriftAnimRateCurrent=~
     SkydiveFidgetAnimRate=~
     SkydiveFidgetAnimRateCurrent=~
     SkydiveAdditiveAlpha=~
     SkydiveDriftDelay=~
    ).bIsParachuteOpen)]
 }
 Transition {
  path=AnimGraph.AnimGraphNode_StateMachine_0.Default.AnimStateNode_3.Skydive.AnimGraphNode_StateMachine_0.SkyDive.AnimStateTransitionNode_1.Transition
  nodes [Result()]
 }
 Transition {
  path=AnimGraph.AnimGraphNode_StateMachine_0.Default.AnimStateNode_3.Skydive.AnimGraphNode_StateMachine_0.SkyDive.AnimStateTransitionNode_2.Transition
  nodes [
   n1 = Get SkydivingState()
   Result(bCanEnterTransition=NOT Boolean(A=Break Fort Anim Input Skydiving(FortAnimInput_Skydiving=@n1) -> (
      ParaGilderRootOffset=~
      ParaGliderRootRotationOverride=~
      bUseParaGlideRootModifier=~
      bIsSkydivingFromLaunchPad=~
      bIsSkydivingFromBus=~
      bIsInVortex=~
      bIsDiving=~
      bIsParachuteOpen=~
      bIsSkydiving=~
      bIsParachuteLeaning=~
      bIsSkydiveLeaning=~
      bIsLeaning=~
      bIsSkydiveDiveMode=~
      bPlaySkydiveDrift=~
      bSkydiveDriftDelayActive=~
      bSkydiveDriftAnimAllowed=~
      bIsGliderRight=~
      bIsGliderCenter=~
      bIsGliderLeft=~
      bIsGliderForward=~
      bIsGliderBack=~
      bHasDeployedAndIsStandingUpright=~
      LocalAccelForward=~
      LocalAccelRight=~
      LocalVelocityRight=~
      SkydiveAimPitch=~
      SkydiveAimYaw=~
      DeployChuteAnimRate=~
      SkydiveDriftAnimRate=~
      SkydiveDriftAnimRateCurrent=~
      SkydiveFidgetAnimRate=~
      SkydiveFidgetAnimRateCurrent=~
      SkydiveAdditiveAlpha=~
      SkydiveDriftDelay=~
     ).bIsParachuteOpen))
  ]
 }
 Transition {
  path=AnimGraph.AnimGraphNode_StateMachine_0.Default.AnimStateNode_3.Skydive.AnimGraphNode_StateMachine_0.SkyDive.AnimStateTransitionNode_3.Transition
  nodes [Result()]
 }
 Transition {
  path=AnimGraph.AnimGraphNode_StateMachine_0.Default.AnimStateNode_3.Skydive.AnimGraphNode_StateMachine_0.SkyDive.AnimStateTransitionNode_4.Transition
  nodes [Result(bCanEnterTransition=Break Fort Anim Input Skydiving(FortAnimInput_Skydiving=Get SkydivingState()) -> (
     ParaGilderRootOffset=~
     ParaGliderRootRotationOverride=~
     bUseParaGlideRootModifier=~
     bIsSkydivingFromLaunchPad=~
     bIsSkydivingFromBus=~
     bIsInVortex=~
     bIsDiving=~
     bIsParachuteOpen=~
     bIsSkydiving=~
     bIsParachuteLeaning=~
     bIsSkydiveLeaning=~
     bIsLeaning=~
     bIsSkydiveDiveMode=~
     bPlaySkydiveDrift=~
     bSkydiveDriftDelayActive=~
     bSkydiveDriftAnimAllowed=~
     bIsGliderRight=~
     bIsGliderCenter=~
     bIsGliderLeft=~
     bIsGliderForward=~
     bIsGliderBack=~
     bHasDeployedAndIsStandingUpright=~
     LocalAccelForward=~
     LocalAccelRight=~
     LocalVelocityRight=~
     SkydiveAimPitch=~
     SkydiveAimYaw=~
     DeployChuteAnimRate=~
     SkydiveDriftAnimRate=~
     SkydiveDriftAnimRateCurrent=~
     SkydiveFidgetAnimRate=~
     SkydiveFidgetAnimRateCurrent=~
     SkydiveAdditiveAlpha=~
     SkydiveDriftDelay=~
    ).bIsDiving)]
 }
 Transition {
  path=AnimGraph.AnimGraphNode_StateMachine_0.Default.AnimStateNode_3.Skydive.AnimGraphNode_StateMachine_0.SkyDive.AnimStateTransitionNode_5.Transition
  nodes [Result()]
 }
 Transition {
  path=AnimGraph.AnimGraphNode_StateMachine_0.Default.AnimStateNode_3.Skydive.AnimGraphNode_StateMachine_0.SkyDive.AnimStateTransitionNode_6.Transition
  nodes [
   n1 = Get SkydivingState()
   Result(bCanEnterTransition=NOT Boolean(A=Break Fort Anim Input Skydiving(FortAnimInput_Skydiving=@n1) -> (
      ParaGilderRootOffset=~
      ParaGliderRootRotationOverride=~
      bUseParaGlideRootModifier=~
      bIsSkydivingFromLaunchPad=~
      bIsSkydivingFromBus=~
      bIsInVortex=~
      bIsDiving=~
      bIsParachuteOpen=~
      bIsSkydiving=~
      bIsParachuteLeaning=~
      bIsSkydiveLeaning=~
      bIsLeaning=~
      bIsSkydiveDiveMode=~
      bPlaySkydiveDrift=~
      bSkydiveDriftDelayActive=~
      bSkydiveDriftAnimAllowed=~
      bIsGliderRight=~
      bIsGliderCenter=~
      bIsGliderLeft=~
      bIsGliderForward=~
      bIsGliderBack=~
      bHasDeployedAndIsStandingUpright=~
      LocalAccelForward=~
      LocalAccelRight=~
      LocalVelocityRight=~
      SkydiveAimPitch=~
      SkydiveAimYaw=~
      DeployChuteAnimRate=~
      SkydiveDriftAnimRate=~
      SkydiveDriftAnimRateCurrent=~
      SkydiveFidgetAnimRate=~
      SkydiveFidgetAnimRateCurrent=~
      SkydiveAdditiveAlpha=~
      SkydiveDriftDelay=~
     ).bIsDiving))
  ]
 }
 Transition {
  path=AnimGraph.AnimGraphNode_StateMachine_0.Default.AnimStateNode_3.Skydive.AnimGraphNode_StateMachine_0.SkyDive.AnimStateTransitionNode_8.Transition
  nodes [
   n1 = Get SkydivingState()
   Result(bCanEnterTransition=NOT Boolean(A=Break Fort Anim Input Skydiving(FortAnimInput_Skydiving=@n1) -> (
      ParaGilderRootOffset=~
      ParaGliderRootRotationOverride=~
      bUseParaGlideRootModifier=~
      bIsSkydivingFromLaunchPad=~
      bIsSkydivingFromBus=~
      bIsInVortex=~
      bIsDiving=~
      bIsParachuteOpen=~
      bIsSkydiving=~
      bIsParachuteLeaning=~
      bIsSkydiveLeaning=~
      bIsLeaning=~
      bIsSkydiveDiveMode=~
      bPlaySkydiveDrift=~
      bSkydiveDriftDelayActive=~
      bSkydiveDriftAnimAllowed=~
      bIsGliderRight=~
      bIsGliderCenter=~
      bIsGliderLeft=~
      bIsGliderForward=~
      bIsGliderBack=~
      bHasDeployedAndIsStandingUpright=~
      LocalAccelForward=~
      LocalAccelRight=~
      LocalVelocityRight=~
      SkydiveAimPitch=~
      SkydiveAimYaw=~
      DeployChuteAnimRate=~
      SkydiveDriftAnimRate=~
      SkydiveDriftAnimRateCurrent=~
      SkydiveFidgetAnimRate=~
      SkydiveFidgetAnimRateCurrent=~
      SkydiveAdditiveAlpha=~
      SkydiveDriftDelay=~
     ).bIsDiving))
  ]
 }
 Transition {
  path=AnimGraph.AnimGraphNode_StateMachine_0.Default.AnimStateNode_3.Skydive.AnimGraphNode_StateMachine_0.SkyDive.AnimStateTransitionNode_9.Transition
  nodes [Result()]
 }
 DBNO {
  path=AnimGraph.AnimGraphNode_StateMachine_0.Default.AnimStateNode_4.DBNO
  nodes [
   AnimGraphNode_StateMachine_0 = DBNO() {is=title AnimGraphNode_StateMachine}
   Output Animation Pose(Result=@AnimGraphNode_StateMachine_0)
  ]
 }
 DBNO {
  entry=DBNO_Start
  path=AnimGraph.AnimGraphNode_StateMachine_0.Default.AnimStateNode_4.DBNO.AnimGraphNode_StateMachine_0.DBNO
  ids=id
  nodes [
   AnimStateNode_0 = DBNO_Start() {is=title AnimStateNode}
   AnimStateNode_1 = DBNO_Idle() {is=title AnimStateNode}
   AnimStateTransitionNode_0 = AnimStateTransitionNode() {bAutomaticRuleBasedOnSequencePlayerInState=true, crossfadeDuration=0}
  ]
  links [{
    dst=DBNO_Idle
    rule=AnimGraph.AnimGraphNode_StateMachine_0.Default.AnimStateNode_4.DBNO.AnimGraphNode_StateMachine_0.DBNO.AnimStateTransitionNode_0.Transition
    src=DBNO_Start
    via=AnimStateTransitionNode_0
   }]
 }
 DBNO_Start {
  path=AnimGraph.AnimGraphNode_StateMachine_0.Default.AnimStateNode_4.DBNO.AnimGraphNode_StateMachine_0.DBNO.AnimStateNode_0.DBNO_Start
  nodes [Output Animation Pose(Result=Sequence Player 'DBNO_Start'!(bLoopAnimation:bool=False) -> (Pose) {bLoopAnimation=false, sequence=/Game/Animation/Game/Pets/HightowerRadish/Backpack/DBNO/DBNO_Start})]
 }
 DBNO_Idle {
  path=AnimGraph.AnimGraphNode_StateMachine_0.Default.AnimStateNode_4.DBNO.AnimGraphNode_StateMachine_0.DBNO.AnimStateNode_1.DBNO_Idle
  nodes [Output Animation Pose(Result=Sequence Player 'DBNO_Idle'() {sequence=/Game/Animation/Game/Pets/HightowerRadish/Backpack/DBNO/DBNO_Idle})]
 }
 Transition {
  path=AnimGraph.AnimGraphNode_StateMachine_0.Default.AnimStateNode_4.DBNO.AnimGraphNode_StateMachine_0.DBNO.AnimStateTransitionNode_0.Transition
  nodes [Result()]
 }
 MeleeReact {
  path=AnimGraph.AnimGraphNode_StateMachine_0.Default.AnimStateNode_5.MeleeReact
  nodes [
   Sequence Player 'Proto_Melee_Loop'() {sequence=/Game/Animation/Game/Pets/HightowerRadish/Proto/Proto_Melee_Loop}
   n1 = Get OwnerLocomotionDirection()
   n2 = Get OwnerLocomotionSpeed()
   n3 = Sequence Player 'Locomotion_FlightLoop_Default'() {
    groupName=IdleSync
    method=SyncGroup
    playRate=1.5
    sequence=&Locomotion_FlightLoop_Default
   }
   Output Animation Pose(Result=Apply Additive(
     Base=@n3
     Additive=Blendspace Player 'Locomotion_LeanPoses_BS'(X=@n1, Y=@n2) {blendSpace=&Locomotion_LeanPoses_BS}
    ) {alphaBoolBlend {blendOption=Linear}})
  ]
 }
 EmoteReactOutro {
  path=AnimGraph.AnimGraphNode_StateMachine_0.Default.AnimStateNode_6.EmoteReactOutro
  nodes [Output Animation Pose(Result=Sequence Player 'Emote_React_Outro'!(bLoopAnimation:bool=False) -> (Pose) {bLoopAnimation=false, sequence=/Game/Animation/Game/Pets/HightowerRadish/Backpack/Emote/Emote_React_Outro})]
 }
 Transition {
  path=AnimGraph.AnimGraphNode_StateMachine_0.Default.AnimStateTransitionNode_0.Transition
  nodes [Result(bCanEnterTransition=Get bIsPlayingEmoteExtra() -> (bIsPlayingEmoteExtra:bool))]
 }
 Transition {
  path=AnimGraph.AnimGraphNode_StateMachine_0.Default.AnimStateTransitionNode_1.Transition
  nodes [Result(bCanEnterTransition=NOT Boolean(A=Get bIsPlayingEmoteExtra()))]
 }
 Transition {
  path=AnimGraph.AnimGraphNode_StateMachine_0.Default.AnimStateTransitionNode_10.Transition
  nodes [Result(bCanEnterTransition=NOT Boolean(A=Get bIsPlayingEmoteExtra()))]
 }
 Transition {path=AnimGraph.AnimGraphNode_StateMachine_0.Default.AnimStateTransitionNode_11.Transition, nodes [Result()]}
 Transition {path=AnimGraph.AnimGraphNode_StateMachine_0.Default.AnimStateTransitionNode_2.Transition, nodes [Result()]}
 Transition {
  path=AnimGraph.AnimGraphNode_StateMachine_0.Default.AnimStateTransitionNode_4.Transition
  nodes [Result(bCanEnterTransition=Get bIsSkydivingInGameOrIsParachuteOpen() -> (bIsSkydivingInGameOrIsParachuteOpen:bool))]
 }
 Transition {
  path=AnimGraph.AnimGraphNode_StateMachine_0.Default.AnimStateTransitionNode_5.Transition
  nodes [Result(bCanEnterTransition=NOT Boolean(A=Get bIsSkydivingInGameOrIsParachuteOpen()))]
 }
 Transition {path=AnimGraph.AnimGraphNode_StateMachine_0.Default.AnimStateTransitionNode_6.Transition, nodes [Result(bCanEnterTransition=Get bIsDBNO())]}
 Transition {
  path=AnimGraph.AnimGraphNode_StateMachine_0.Default.AnimStateTransitionNode_7.Transition
  nodes [Result(bCanEnterTransition=NOT Boolean(A=Get bIsDBNO()))]
 }
 Transition {
  path=AnimGraph.AnimGraphNode_StateMachine_0.Default.AnimStateTransitionNode_9.Transition
  nodes [Result(bCanEnterTransition=NOT Boolean(A=Get bIsOwnerPlayingMeleeAnim()))]
 }
 FrontEnd {
  entry=Idle
  path=AnimGraph.AnimGraphNode_StateMachine_1.FrontEnd
  ids=id
  nodes [
   AnimStateNode_0 = Idle()
   AnimStateNode_1 = SkyDive() {is=title AnimStateNode}
   AnimStateTransitionNode_0 = AnimStateTransitionNode()
   AnimStateTransitionNode_1 = AnimStateTransitionNode()
  ]
  links [
   {
    dst=SkyDive
    rule=AnimGraph.AnimGraphNode_StateMachine_1.FrontEnd.AnimStateTransitionNode_0.Transition
    src=Idle
    via=AnimStateTransitionNode_0
   }
   {
    dst=Idle
    rule=AnimGraph.AnimGraphNode_StateMachine_1.FrontEnd.AnimStateTransitionNode_1.Transition
    src=SkyDive
    via=AnimStateTransitionNode_1
   }
  ]
 }
 Idle {
  path=AnimGraph.AnimGraphNode_StateMachine_1.FrontEnd.AnimStateNode_0.Idle
  nodes [
   Sequence Player 'Hover_Idle_NonTargeted'() {
    groupName=IdleSync
    method=SyncGroup
    sequence=&Hover_Idle_NonTargeted
    startPosition=1
   }
   AnimGraphNode_StateMachine_0 = RandomIdle() {is=title AnimGraphNode_StateMachine}
   Output Animation Pose(Result=@AnimGraphNode_StateMachine_0)
  ]
 }
 RandomIdle {
  entry=IdleFacingFwd
  path=AnimGraph.AnimGraphNode_StateMachine_1.FrontEnd.AnimStateNode_0.Idle.AnimGraphNode_StateMachine_0.RandomIdle
  ids=id
  nodes [
   AnimStateNode_0 = IdleFacingFwd() {is=title AnimStateNode}
   AnimStateNode_1 = FwdToBwd() {is=title AnimStateNode, stateEntered {guid=#8, notifyName=FwdToBwd_Entered}}
   AnimStateNode_2 = IdleFacingBwd() {is=title AnimStateNode}
   AnimStateNode_3 = BwdToFwd() {is=title AnimStateNode, stateEntered {guid=#9, notifyName=BwdToFwd_Entered}}
   AnimStateNode_4 = LockerHoverIdle() {is=title AnimStateNode}
   AnimStateTransitionNode_0 = AnimStateTransitionNode()
   AnimStateTransitionNode_1 = AnimStateTransitionNode()
   AnimStateTransitionNode_6 = AnimStateTransitionNode() {crossfadeDuration=0.4}
   AnimStateTransitionNode_7 = AnimStateTransitionNode() {bAutomaticRuleBasedOnSequencePlayerInState=true, crossfadeDuration=0.4}
   AnimStateTransitionNode_8 = AnimStateTransitionNode() {crossfadeDuration=0.4}
   AnimStateTransitionNode_9 = AnimStateTransitionNode() {bAutomaticRuleBasedOnSequencePlayerInState=true, crossfadeDuration=0.4}
  ]
  links [
   {
    dst=IdleFacingFwd
    rule=AnimGraph.AnimGraphNode_StateMachine_1.FrontEnd.AnimStateNode_0.Idle.AnimGraphNode_StateMachine_0.RandomIdle.AnimStateTransitionNode_9.Transition
    src=BwdToFwd
    via=AnimStateTransitionNode_9
   }
   {
    dst=IdleFacingBwd
    rule=AnimGraph.AnimGraphNode_StateMachine_1.FrontEnd.AnimStateNode_0.Idle.AnimGraphNode_StateMachine_0.RandomIdle.AnimStateTransitionNode_7.Transition
    src=FwdToBwd
    via=AnimStateTransitionNode_7
   }
   {
    dst=BwdToFwd
    rule=AnimGraph.AnimGraphNode_StateMachine_1.FrontEnd.AnimStateNode_0.Idle.AnimGraphNode_StateMachine_0.RandomIdle.AnimStateTransitionNode_8.Transition
    src=IdleFacingBwd
    via=AnimStateTransitionNode_8
   }
   {
    dst=FwdToBwd
    rule=AnimGraph.AnimGraphNode_StateMachine_1.FrontEnd.AnimStateNode_0.Idle.AnimGraphNode_StateMachine_0.RandomIdle.AnimStateTransitionNode_6.Transition
    src=IdleFacingFwd
    via=AnimStateTransitionNode_6
   }
   {
    dst=LockerHoverIdle
    rule=AnimGraph.AnimGraphNode_StateMachine_1.FrontEnd.AnimStateNode_0.Idle.AnimGraphNode_StateMachine_0.RandomIdle.AnimStateTransitionNode_0.Transition
    src=IdleFacingFwd
    via=AnimStateTransitionNode_0
   }
   {
    dst=IdleFacingFwd
    rule=AnimGraph.AnimGraphNode_StateMachine_1.FrontEnd.AnimStateNode_0.Idle.AnimGraphNode_StateMachine_0.RandomIdle.AnimStateTransitionNode_1.Transition
    src=LockerHoverIdle
    via=AnimStateTransitionNode_1
   }
  ]
 }
 IdleFacingFwd {
  path=AnimGraph.AnimGraphNode_StateMachine_1.FrontEnd.AnimStateNode_0.Idle.AnimGraphNode_StateMachine_0.RandomIdle.AnimStateNode_0.IdleFacingFwd
  nodes [
   Sequence Player 'Hover_Idle_FrontEnd'() {sequence=/Game/Animation/Game/Pets/HightowerRadish/Backpack/Idle/Hover_Idle_FrontEnd}
   n1 = Sequence Player 'Hover_Idle_NonTargeted'() {
    groupName=IdleSync
    method=SyncGroup
    sequence=&Hover_Idle_NonTargeted
    startPosition=1
   }
   Output Animation Pose(Result=@n1)
  ]
 }
 FwdToBwd {
  path=AnimGraph.AnimGraphNode_StateMachine_1.FrontEnd.AnimStateNode_0.Idle.AnimGraphNode_StateMachine_0.RandomIdle.AnimStateNode_1.FwdToBwd
  nodes [Output Animation Pose(Result=Sequence Player 'Locomotion_FlightLoop_Intro'!(bLoopAnimation:bool=False) -> (Pose) {bLoopAnimation=false, sequence=&Locomotion_FlightLoop_Intro})]
 }
 IdleFacingBwd {
  path=AnimGraph.AnimGraphNode_StateMachine_1.FrontEnd.AnimStateNode_0.Idle.AnimGraphNode_StateMachine_0.RandomIdle.AnimStateNode_2.IdleFacingBwd
  nodes [Output Animation Pose(Result=Sequence Player 'Locomotion_FlightLoop_Default'() {sequence=&Locomotion_FlightLoop_Default})]
 }
 BwdToFwd {
  path=AnimGraph.AnimGraphNode_StateMachine_1.FrontEnd.AnimStateNode_0.Idle.AnimGraphNode_StateMachine_0.RandomIdle.AnimStateNode_3.BwdToFwd
  nodes [Output Animation Pose(Result=Sequence Player 'Locomotion_FlightLoop_Exit'!(bLoopAnimation:bool=False) -> (Pose) {bLoopAnimation=false, sequence=&Locomotion_FlightLoop_Exit})]
 }
 LockerHoverIdle {
  path=AnimGraph.AnimGraphNode_StateMachine_1.FrontEnd.AnimStateNode_0.Idle.AnimGraphNode_StateMachine_0.RandomIdle.AnimStateNode_4.LockerHoverIdle
  nodes [Output Animation Pose(Result=Sequence Player 'Hover_Idle_FrontEnd'() {sequence=&Hover_Idle_FrontEnd})]
 }
 Transition {
  path=AnimGraph.AnimGraphNode_StateMachine_1.FrontEnd.AnimStateNode_0.Idle.AnimGraphNode_StateMachine_0.RandomIdle.AnimStateTransitionNode_0.Transition
  nodes [Result(bCanEnterTransition=NOT Boolean(A=Get bIsFrontEndLobbyPartyMember() -> (bIsFrontEndLobbyPartyMember:bool)))]
 }
 Transition {
  path=AnimGraph.AnimGraphNode_StateMachine_1.FrontEnd.AnimStateNode_0.Idle.AnimGraphNode_StateMachine_0.RandomIdle.AnimStateTransitionNode_1.Transition
  nodes [
   Result(bCanEnterTransition=Get bIsFrontEndLobbyPartyMember())
   Get bTransition_Idle_IdleLobby() -> (bTransition_Idle_IdleLobby:bool)
  ]
 }
 Transition {
  path=AnimGraph.AnimGraphNode_StateMachine_1.FrontEnd.AnimStateNode_0.Idle.AnimGraphNode_StateMachine_0.RandomIdle.AnimStateTransitionNode_6.Transition
  nodes [Result(bCanEnterTransition=Get bFrontEndIdleTransition() -> (bFrontEndIdleTransition:bool))]
 }
 Transition {
  path=AnimGraph.AnimGraphNode_StateMachine_1.FrontEnd.AnimStateNode_0.Idle.AnimGraphNode_StateMachine_0.RandomIdle.AnimStateTransitionNode_7.Transition
  nodes [Result()]
 }
 Transition {
  path=AnimGraph.AnimGraphNode_StateMachine_1.FrontEnd.AnimStateNode_0.Idle.AnimGraphNode_StateMachine_0.RandomIdle.AnimStateTransitionNode_8.Transition
  nodes [Result(bCanEnterTransition=Get bFrontEndIdleTransition())]
 }
 Transition {
  path=AnimGraph.AnimGraphNode_StateMachine_1.FrontEnd.AnimStateNode_0.Idle.AnimGraphNode_StateMachine_0.RandomIdle.AnimStateTransitionNode_9.Transition
  nodes [Result()]
 }
 SkyDive {
  path=AnimGraph.AnimGraphNode_StateMachine_1.FrontEnd.AnimStateNode_1.SkyDive
  nodes [
   AnimGraphNode_StateMachine_0 = SkyDiveGlideOrDive() {is=title AnimGraphNode_StateMachine}
   Output Animation Pose(Result=Apply Additive(
     Base=Sequence Player 'Glide_Pose'() {sequence=/Game/Animation/Game/Pets/HightowerRadish/Backpack/Glider/Glide_Pose}
     Additive="Sequence Player (additive) 'Glide'"() {sequence=&Glide}
    ) {alphaBoolBlend {blendOption=Linear}})
  ]
 }
 SkyDiveGlideOrDive {
  entry=Glide
  path=AnimGraph.AnimGraphNode_StateMachine_1.FrontEnd.AnimStateNode_1.SkyDive.AnimGraphNode_StateMachine_0.SkyDiveGlideOrDive
  ids=id
  nodes [
   AnimStateNode_0 = Glide()
   AnimStateNode_1 = GlideToDive() {is=title AnimStateNode}
   AnimStateNode_2 = Dive() {is=title AnimStateNode}
   AnimStateNode_3 = DiveToGlide()
   AnimStateTransitionNode_2 = AnimStateTransitionNode()
   AnimStateTransitionNode_3 = AnimStateTransitionNode() {bAutomaticRuleBasedOnSequencePlayerInState=true}
   AnimStateTransitionNode_4 = AnimStateTransitionNode()
   AnimStateTransitionNode_5 = AnimStateTransitionNode() {bAutomaticRuleBasedOnSequencePlayerInState=true}
  ]
  links [
   {
    dst=DiveToGlide
    rule=AnimGraph.AnimGraphNode_StateMachine_1.FrontEnd.AnimStateNode_1.SkyDive.AnimGraphNode_StateMachine_0.SkyDiveGlideOrDive.AnimStateTransitionNode_4.Transition
    src=Dive
    via=AnimStateTransitionNode_4
   }
   {
    dst=Glide
    rule=AnimGraph.AnimGraphNode_StateMachine_1.FrontEnd.AnimStateNode_1.SkyDive.AnimGraphNode_StateMachine_0.SkyDiveGlideOrDive.AnimStateTransitionNode_5.Transition
    src=DiveToGlide
    via=AnimStateTransitionNode_5
   }
   {
    dst=GlideToDive
    rule=AnimGraph.AnimGraphNode_StateMachine_1.FrontEnd.AnimStateNode_1.SkyDive.AnimGraphNode_StateMachine_0.SkyDiveGlideOrDive.AnimStateTransitionNode_2.Transition
    src=Glide
    via=AnimStateTransitionNode_2
   }
   {
    dst=Dive
    rule=AnimGraph.AnimGraphNode_StateMachine_1.FrontEnd.AnimStateNode_1.SkyDive.AnimGraphNode_StateMachine_0.SkyDiveGlideOrDive.AnimStateTransitionNode_3.Transition
    src=GlideToDive
    via=AnimStateTransitionNode_3
   }
  ]
 }
 Glide {
  path=AnimGraph.AnimGraphNode_StateMachine_1.FrontEnd.AnimStateNode_1.SkyDive.AnimGraphNode_StateMachine_0.SkyDiveGlideOrDive.AnimStateNode_0.Glide
  nodes [Output Animation Pose(Result=Sequence Player 'Paraglide'() {sequence=&Paraglide})]
 }
 GlideToDive {
  path=AnimGraph.AnimGraphNode_StateMachine_1.FrontEnd.AnimStateNode_1.SkyDive.AnimGraphNode_StateMachine_0.SkyDiveGlideOrDive.AnimStateNode_1.GlideToDive
  nodes [Output Animation Pose(Result=Sequence Player 'Paraglide_To_Glide'!(bLoopAnimation:bool=False) -> (Pose) {bLoopAnimation=false, sequence=&Paraglide_To_Glide})]
 }
 Dive {
  path=AnimGraph.AnimGraphNode_StateMachine_1.FrontEnd.AnimStateNode_1.SkyDive.AnimGraphNode_StateMachine_0.SkyDiveGlideOrDive.AnimStateNode_2.Dive
  nodes [Output Animation Pose(Result="Sequence Player (additive) 'Glide'"() {sequence=&Glide})]
 }
 DiveToGlide {
  path=AnimGraph.AnimGraphNode_StateMachine_1.FrontEnd.AnimStateNode_1.SkyDive.AnimGraphNode_StateMachine_0.SkyDiveGlideOrDive.AnimStateNode_3.DiveToGlide
  nodes [Output Animation Pose(Result=Sequence Player 'Glide_To_Paraglide'!(bLoopAnimation:bool=False) -> (Pose) {bLoopAnimation=false, sequence=&Glide_To_Paraglide})]
 }
 Transition {
  path=AnimGraph.AnimGraphNode_StateMachine_1.FrontEnd.AnimStateNode_1.SkyDive.AnimGraphNode_StateMachine_0.SkyDiveGlideOrDive.AnimStateTransitionNode_2.Transition
  nodes [Result(bCanEnterTransition=Get bHighTowerRadish_LobbySkyDive_Diving() -> (bHighTowerRadish_LobbySkyDive_Diving:bool))]
 }
 Transition {
  path=AnimGraph.AnimGraphNode_StateMachine_1.FrontEnd.AnimStateNode_1.SkyDive.AnimGraphNode_StateMachine_0.SkyDiveGlideOrDive.AnimStateTransitionNode_3.Transition
  nodes [Result()]
 }
 Transition {
  path=AnimGraph.AnimGraphNode_StateMachine_1.FrontEnd.AnimStateNode_1.SkyDive.AnimGraphNode_StateMachine_0.SkyDiveGlideOrDive.AnimStateTransitionNode_4.Transition
  nodes [Result(bCanEnterTransition=Get bHighTowerRadish_LobbySkyDive_Gliding() -> (bHighTowerRadish_LobbySkyDive_Gliding:bool))]
 }
 Transition {
  path=AnimGraph.AnimGraphNode_StateMachine_1.FrontEnd.AnimStateNode_1.SkyDive.AnimGraphNode_StateMachine_0.SkyDiveGlideOrDive.AnimStateTransitionNode_5.Transition
  nodes [Result()]
 }
 Transition {
  path=AnimGraph.AnimGraphNode_StateMachine_1.FrontEnd.AnimStateTransitionNode_0.Transition
  nodes [Result(bCanEnterTransition=Get bIsSkydivingFE() -> (bIsSkydivingFE:bool))]
 }
 Transition {
  path=AnimGraph.AnimGraphNode_StateMachine_1.FrontEnd.AnimStateTransitionNode_1.Transition
  nodes [Result(bCanEnterTransition=NOT Boolean(A=Get bIsSkydivingFE()))]
 }
 EventGraph {nodes [
   HighTowerRadish_Lobby SkyDive Diving() {is=title K2Node_CustomEvent}
   > Set bHighTowerRadish_LobbySkyDive_Diving(bHighTowerRadish_LobbySkyDive_Diving:bool=true) -> (Output_Get:bool)
   > Set bHighTowerRadish_LobbySkyDive_Gliding(bHighTowerRadish_LobbySkyDive_Gliding:bool=false) -> (Output_Get:bool)
   HighTowerRadish_Lobby SkyDive Gliding() {is=title K2Node_CustomEvent}
   > Set bHighTowerRadish_LobbySkyDive_Gliding(bHighTowerRadish_LobbySkyDive_Gliding=true)
   > Set bHighTowerRadish_LobbySkyDive_Diving(bHighTowerRadish_LobbySkyDive_Diving=false)
   Event BlueprintUpdateAnimation()
   > n1 = Branch(Condition=Get bIsFrontEndPreview())
   Is Valid(exec=@n1.else, InputObject=Get OwnerFortPawn() -> (OwnerFortPawn:Fort Player Pawn Object Reference))
   n2 = Get OwnerAnimBP()
   > Is Valid(InputObject=@n2)
   > Set IsDBNOCarriedAlpha(IsDBNOCarriedAlpha:double="To Float (Boolean)"(InBool=Get bIsDBNOCarried!(self:Fort Player Anim Instance Object Reference=@n2) -> (bIsDBNOCarried))) -> (Output_Get:double)
   > n3 = Sequence!(execute) -> (then_0, then_1, then_2:Exec, then_3:Exec, then_4:Exec)
   n4 = Get OwnerAnimBP()
   > Branch(Condition=OR Boolean!(A=Get bIsMoving2D(), B=Get bIsPlayingMeleeAnim(self=@n4), C:bool=Get bIsTargeting()) -> (ReturnValue))
   > Set bStoppedMovingOrMeleeForLongEnoughTime(bStoppedMovingOrMeleeForLongEnoughTime:bool=false) -> (Output_Get:bool)
   > RetriggerableDelay(Duration=2)
   > Set bStoppedMovingOrMeleeForLongEnoughTime(bStoppedMovingOrMeleeForLongEnoughTime=true)
   Branch(execute=@n3.then_2, Condition=IsTurningSharply!(self:Fort Pet Anim Instance High Tower Radish Object Reference) -> (ReturnValue:bool))
   > Set bSharpDirectionChange(bSharpDirectionChange:bool=true) -> (Output_Get:bool)
   > RetriggerableDelay(Duration=0.2)
   > Set bSharpDirectionChange(bSharpDirectionChange=false)
   Branch(execute=@n3.then_1, Condition=Get bIsTargeting())
   > Set bHasTargettedInLastNSeconds(bHasTargettedInLastNSeconds:bool=true) -> (Output_Get:bool)
   > RetriggerableDelay(Duration=2.5)
   > Set bHasTargettedInLastNSeconds(bHasTargettedInLastNSeconds=false)
   n5 = Get bIsPlayingMeleeAnim(self=Get OwnerAnimBP())
   Branch(execute=@n3.then_3, Condition=AND Boolean(A=NOT Boolean(A=@n5), B=Get bIsPlayingMeleeAnimLastTick() -> (bIsPlayingMeleeAnimLastTick:bool)))
   > Set bSlowDownInterp(bSlowDownInterp:bool=true) -> (Output_Get:bool)
   > RetriggerableDelay(Duration=1.2)
   > Set bSlowDownInterp(bSlowDownInterp=false)
   Set bIsPlayingMeleeAnimLastTick(execute=@n3.then_4, bIsPlayingMeleeAnimLastTick:bool=Get bIsPlayingMeleeAnim(self=Get OwnerAnimBP())) -> (Output_Get:bool)
   AnimNotify_RandomIdlePick() {is=title K2Node_Event}
   > Set bFrontEndIdleTransition(bFrontEndIdleTransition:bool=RandomBool!() -> (ReturnValue:bool)) -> (Output_Get:bool)
   n6 = AnimNotify_BwdToFwd_Entered() {is=title K2Node_Event}
   Set bFrontEndIdleTransition(execute=[@n6.then, @n7.then], bFrontEndIdleTransition=false)
   n7 = AnimNotify_FwdToBwd_Entered() {is=title K2Node_Event}
   AnimNotify_TriggerRandomizer() {is=title K2Node_Event}
   > Set RandomLocomotionBreak(RandomLocomotionBreak:int=RandomInteger!(Max:int=2) -> (ReturnValue:int=0)) -> (Output_Get:int=0)
   > Set bAllowRandomLocomotionBreak(bAllowRandomLocomotionBreak:bool=RandomBoolWithWeight!(Weight:float=1) -> (ReturnValue:bool)) -> (Output_Get:bool)
   AnimNotify_MoveLoopRandomEntered() {is=title K2Node_Event}
   > Set bAllowRandomLocomotionBreak(bAllowRandomLocomotionBreak=false)
   Event OnEnteredVehicle() {is=title K2Node_Event}
   > n8 = Cast To B_PetInst_HightowerRadish!(execute, Object=Get OwnerPet() -> (OwnerPet:Fort Player Pet Object Reference)) -> (then, CastFailed, AsB Pet Inst Hightower Radish:B Pet Inst Hightower Radish Object Reference) {is=title K2Node_DynamicCast}
   > SetHiddenInGame!(
    execute
    self:Scene Component Object Reference=Get Niagara_Radish_Jets!(self:B Pet Inst Hightower Radish Object Reference=@n8) -> (Niagara_Radish_Jets:Niagara Particle System Component Object Reference)
    NewHidden:bool=true
    bPropagateToChildren:bool=false
   ) -> (then)
   Event OnExitedVehicle() {is=title K2Node_Event}
   > n9 = Cast To B_PetInst_HightowerRadish(Object=Get OwnerPet())
   > SetHiddenInGame(self=Get Niagara_Radish_Jets(self=@n9), NewHidden=false, bPropagateToChildren=false)
  ]}
]
variables [
 bAllowRandomLocomotionBreak {type=boolean}
 bFrontEndIdleTransition {type=boolean}
 bHasTargettedInLastNSeconds {type=boolean}
 bHighTowerRadish_LobbySkyDive_Diving {type=boolean}
 bHighTowerRadish_LobbySkyDive_Gliding {type=boolean}
 bIsPlayingMeleeAnimLastTick {type=boolean}
 bSharpDirectionChange {type=boolean}
 bStoppedMovingOrMeleeForLongEnoughTime {type=boolean, default=true}
 IsDBNOCarriedAlpha {type=number}
 RandomLocomotionBreak {type=integer}
]
