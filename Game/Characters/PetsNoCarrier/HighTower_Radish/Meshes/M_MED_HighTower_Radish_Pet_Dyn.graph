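# Physics asset for the HighTower Radish pet, covering its root, pelvis and a four-bone tail.
# Each bone has one capsule: short, wide ones on root and pelvis, identical thin ones along the tail.
# Constraints chain the tail to the pelvis with tight twist limits and swing that is widest between tail_01 and tail_02, stiffest toward the tip.
# The pelvis rotates freely on all angular axes relative to root.
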PhysicsAsset /Game/Characters/PetsNoCarrier/HighTower_Radish/Meshes/M_MED_HighTower_Radish_Pet_Dyn
bodies {
 pelvis [{shapeType=Capsule, length=0.505, radius=6.63}]
 root [{shapeType=Capsule, length=0.505, radius=6.63}]
 tail_01 [{shapeType=Capsule, length=4.9062, radius=1, center=<0 -2.9705 0>, rotation=<rot 0 0 90>}]
 tail_02 [{shapeType=Capsule, length=4.9062, radius=1, center=<0 -2.9705 0>, rotation=<rot 0 0 90>}]
 tail_03 [{shapeType=Capsule, length=4.9062, radius=1, center=<0 -2.9705 0>, rotation=<rot 0 0 90>}]
 tail_04 [{shapeType=Capsule, length=4.9062, radius=1, center=<0 -2.9705 0>, rotation=<rot 0 0 90>}]
}
constraints [
 {bone1Name=tail_04, bone2Name=tail_03, swing1LimitDegrees=5, swing2LimitDegrees=5, twistLimitDegrees=5}
 {bone1Name=tail_02, bone2Name=tail_01, swing1LimitDegrees=30, swing2LimitDegrees=30, twistLimitDegrees=5}
 {bone1Name=tail_03, bone2Name=tail_02, swing1LimitDegrees=5, swing2LimitDegrees=5, twistLimitDegrees=5}
 {bone1Name=tail_01, bone2Name=pelvis, swing1LimitDegrees=15, swing2LimitDegrees=15, twistLimitDegrees=1}
 {
  bone1Name=pelvis
  bone2Name=root
  swing1LimitDegrees=45
  swing1Motion=Free
  swing2LimitDegrees=45
  swing2Motion=Free
  twistLimitDegrees=45
  twistMotion=Free
 }
]
