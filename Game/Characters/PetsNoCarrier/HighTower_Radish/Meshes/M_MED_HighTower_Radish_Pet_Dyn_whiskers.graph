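# Physics asset for the whiskers of the HighTower Radish pet, defining capsule bodies for dynamic whisker motion.
# Left and right whisker capsules each attach to the head with a limited swing cone and almost no twist.
# The head, pelvis and root capsules form the supporting chain, with free angular motion between them.
# The pelvis capsule is by far the largest, offset and rotated as a body volume.
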PhysicsAsset /Game/Characters/PetsNoCarrier/HighTower_Radish/Meshes/M_MED_HighTower_Radish_Pet_Dyn_whiskers
bodies {
 head [{shapeType=Capsule, length=0.5, radius=0.5}]
 pelvis [{shapeType=Capsule, length=7.0617, radius=5.6955, center=<-0.5608 -1.5875 -0.2302>, rotation=<rot 0 -10.8887 -60.6925>}]
 root [{shapeType=Capsule, length=0.505, radius=0.505, rotation=<rot 0 -90 0>}]
 whiskers_l [{shapeType=Capsule, length=2, radius=1, center=<2.5 0 0>, rotation=<rot 90 0 0>}]
 whiskers_r [{shapeType=Capsule, length=2, radius=1, center=<-2.5 0.0002 0>, rotation=<rot 90 0 0>}]
}
constraints [
 {bone1Name=whiskers_r, bone2Name=head, swing1LimitDegrees=35, swing2LimitDegrees=35, twistLimitDegrees=1}
 {bone1Name=whiskers_l, bone2Name=head, swing1LimitDegrees=35, swing2LimitDegrees=35, twistLimitDegrees=1}
 {
  bone1Name=head
  bone2Name=pelvis
  swing1LimitDegrees=45
  swing1Motion=Free
  swing2LimitDegrees=45
  swing2Motion=Free
  twistLimitDegrees=45
  twistMotion=Free
 }
 {
  bone1Name=pelvis
  bone2Name=root
  swing1LimitDegrees=45
  swing1Motion=Free
  swing2LimitDegrees=45
  swing2Motion=Free
  twistLimitDegrees=45
  twistMotion=Free
 }
]
